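# Physics asset for the F_MED_RollerBlade jacket face accessory (front end), defining simulation shapes for the character's right-front hair strand.
# A chain of boxes and capsules runs from a head-attached root, forking at hair_r_front_04 into two short tip branches.
# Angular limits are tight near the root and generally widen down the chain, with the 08-09 branch stiffer.
# Head, neck, upper spine, right clavicle and jacket flap bodies, joined by free-rotation constraints, supply surrounding body shapes.
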PhysicsAsset /BRCosmetics/Characters/Player/Female/Medium/Bodies/F_MED_RollerBlade/Meshes/Parts/F_MED_RollerBladeJacket_FaceAcc_Front_FrontEnd_Physics
bodies {
 clavicle_r [{
   shapeType=Capsule
   length=10
   radius=7.9425
   shapeName=clavicle_r_capsule
   center=<-5.8727 3.9491 0.4724>
   rotation=<rot 76.5912 -0.0005 -0.0005>
  }]
 dyn_jacket_r_flap_upper_01 [{
   shapeType=Box
   extentX=10.625
   extentY=4.8125
   extentZ=20.75
   shapeName=dyn_jacket_r_flap_upper_01_box
   center=<-2.5041 2.5921 -1.0989>
   rotation=<rot 6.0919 6.3716 4.2342>
  }]
 hair_r_front_01 [{shapeType=Box, extentX=7.5, extentY=2.375, extentZ=7.5, shapeName=hair_r_front_01_box, center=<2.9194 0 1.2934>}]
 hair_r_front_01_root [{shapeType=Capsule, length=0.505, radius=0.505, shapeName=hair_r_front_01_root_capsule, rotation=<rot 0 -90 0>}]
 hair_r_front_02 [{
   shapeType=Box
   extentX=7.5
   extentY=2.375
   extentZ=7.5
   shapeName=hair_r_front_02_box
   center=<0.5905 -1.3408 0>
   rotation=<rot -0.0002 -9.2369 0.0001>
  }]
 hair_r_front_03 [{shapeType=Box, extentX=7.5, extentY=2.75, extentZ=7.5, shapeName=hair_r_front_03_box, center=<0 -0.9618 0>}]
 hair_r_front_04 [{
   shapeType=Capsule
   length=3.842
   radius=2.6408
   shapeName=hair_r_front_04_capsule
   center=<0.2852 -0.3801 -0.2374>
   rotation=<rot 62.8381 -66.3369 -86.9541>
  }]
 hair_r_front_05 [{
   shapeType=Capsule
   length=1.8438
   radius=2.5231
   shapeName=hair_r_front_05_capsule
   center=<-0.0398 0.3545 -0.4273>
   rotation=<rot 90 180 173.5903>
  }]
 hair_r_front_06 [{shapeType=Capsule, length=3.8831, radius=1.6123, shapeName=hair_r_front_06_capsule, rotation=<rot 90 0 6.6769>}]
 hair_r_front_07 [{shapeType=Capsule, length=5.8186, radius=1.3228, shapeName=hair_r_front_07_capsule, center=<2.5054 0 0>, rotation=<rot 90 101.5877 87.3914>}]
 hair_r_front_08 [{shapeType=Capsule, length=3.3163, radius=1.7713, shapeName=hair_r_front_08_capsule, center=<1.9991 -0.593 0>, rotation=<rot 90 0 0>}]
 hair_r_front_09 [{shapeType=Capsule, length=3.3499, radius=1.318, shapeName=hair_r_front_09_capsule, center=<1.2747 0.1937 0>, rotation=<rot 90 0 -17.4878>}]
 head [{shapeType=Capsule, length=7.7033, radius=5.4414, center=<0.9666 4.1136 0>, rotation=<rot 90 0 -12.383>}]
 neck_01 [{shapeType=Capsule, length=13.255, radius=5.13, shapeName=neck_01_capsule, center=<0 0.4367 0>, rotation=<rot 0 -90.0011 -89.998>}]
 spine_04 [{shapeType=Capsule, length=14.0938, radius=10.7188, shapeName=spine_04_capsule, center=<-0.4216 1.619 0>, rotation=<rot 0 -90 0>}]
 spine_05 [{shapeType=Capsule, length=13.5988, radius=7.9738, shapeName=spine_05_capsule, center=<1.5653 -1.1893 0>, rotation=<rot 0 -90 0>}]
}
constraints [
 {
  bone1Name=hair_r_front_01_root
  bone2Name=head
  swing1LimitDegrees=45
  swing1Motion=Free
  swing2LimitDegrees=45
  swing2Motion=Free
  twistLimitDegrees=45
  twistMotion=Free
 }
 {bone1Name=hair_r_front_01, bone2Name=hair_r_front_01_root, swing1LimitDegrees=4, swing2LimitDegrees=4, twistLimitDegrees=3}
 {bone1Name=hair_r_front_02, bone2Name=hair_r_front_01, swing1LimitDegrees=4, swing2LimitDegrees=4, twistLimitDegrees=3}
 {bone1Name=hair_r_front_03, bone2Name=hair_r_front_02, swing1LimitDegrees=4, swing2LimitDegrees=4, twistLimitDegrees=3}
 {bone1Name=hair_r_front_04, bone2Name=hair_r_front_03, swing1LimitDegrees=24, swing2LimitDegrees=24, twistLimitDegrees=8}
 {bone1Name=hair_r_front_05, bone2Name=hair_r_front_04, swing1LimitDegrees=24, swing2LimitDegrees=24, twistLimitDegrees=24}
 {bone1Name=hair_r_front_06, bone2Name=hair_r_front_05, swing1LimitDegrees=32, swing2LimitDegrees=32, twistLimitDegrees=32}
 {bone1Name=hair_r_front_07, bone2Name=hair_r_front_06, swing1LimitDegrees=32, swing2LimitDegrees=32, twistLimitDegrees=32}
 {bone1Name=hair_r_front_08, bone2Name=hair_r_front_04, swing1LimitDegrees=8, swing2LimitDegrees=8, twistLimitDegrees=8}
 {bone1Name=hair_r_front_09, bone2Name=hair_r_front_08, swing1LimitDegrees=10, swing2LimitDegrees=10, twistLimitDegrees=10}
 {
  bone1Name=head
  bone2Name=neck_01
  swing1LimitDegrees=45
  swing1Motion=Free
  swing2LimitDegrees=45
  swing2Motion=Free
  twistLimitDegrees=45
  twistMotion=Free
 }
 {
  bone1Name=spine_05
  bone2Name=spine_04
  swing1LimitDegrees=45
  swing1Motion=Free
  swing2LimitDegrees=45
  swing2Motion=Free
  twistLimitDegrees=45
  twistMotion=Free
 }
 {
  bone1Name=neck_01
  bone2Name=spine_05
  swing1LimitDegrees=45
  swing1Motion=Free
  swing2LimitDegrees=45
  swing2Motion=Free
  twistLimitDegrees=45
  twistMotion=Free
 }
 {
  bone1Name=clavicle_r
  bone2Name=spine_05
  swing1LimitDegrees=45
  swing1Motion=Free
  swing2LimitDegrees=45
  swing2Motion=Free
  twistLimitDegrees=45
  twistMotion=Free
 }
 {
  bone1Name=dyn_jacket_r_flap_upper_01
  bone2Name=spine_04
  swing1LimitDegrees=45
  swing1Motion=Free
  swing2LimitDegrees=45
  swing2Motion=Free
  twistLimitDegrees=45
  twistMotion=Free
 }
]
