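# Control Rig for the F_MED_RollerBlade face accessory mesh that drives corrective hair motion from body pose.
# Construction snaps controls to front hair bones and caches bind and offset target poses.
# Each update, spherical pose readers on head, right upper arm and clavicle weight pose blends, switched by IsJacketVariant.
# Weighted parent constraints anchor front-hair nulls, root offsets and aim constraints steer side/rear hair, and Spline IK fits the right front chain.
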ControlRigBlueprint /BRCosmetics/Characters/Player/Female/Medium/Bodies/F_MED_RollerBlade/Meshes/Parts/F_MED_RollerBlade_FaceAcc_CtrlRig
graphs [
 New Function_ContainedGraph {nodes [
   n1 = RigVMFunctionEntryNode()
   > RigVMVariableNode(Value=@n1.controls) {variable=local_controls}
   > RigVMVariableNode(Value=@n1.bones) {variable=local_bones}
   > RigVMVariableNode(Value=@n1.primary axis) {variable=local_primary_axis}
   > RigVMVariableNode(Value=@n1.samples) {variable=local_samples}
   > RigVMVariableNode(Value=@n1.stretch) {variable=local_stretch}
   > RigVMVariableNode(Value=@n1.debug draw spline) {variable=local_debug_draw}
   > RigVMVariableNode(Value=@n1.max stretch) {variable=local_max stretch}
   > RigVMVariableNode(Value=@n1.min stretch) {variable=local_min stretch}
   > DISPATCH_RigVMDispatch_ArrayReset = RigVMDispatchNode(Array=RigVMVariableNode() {variable=points}.Value) {template="DISPATCH_RigVMDispatch_ArrayReset(io Array)"}
   > DISPATCH_RigVMDispatch_ArrayIterator = RigVMDispatchNode(Array=RigVMVariableNode() {variable=local_controls}.Value) {template="DISPATCH_RigVMDispatch_ArrayIterator(in Array,out Element,out Index,out Count,out Ratio)"}
   > DISPATCH_RigVMDispatch_ArrayAdd = RigVMDispatchNode(
    Element=&RigUnit_GetTransform(Item=@DISPATCH_RigVMDispatch_ArrayIterator.Element).Transform.Translation
    Array=RigVMVariableNode() {variable=points}.Value
   ) {template="DISPATCH_RigVMDispatch_ArrayAdd(io Array,in Element,out Index)"}
   RigVMVariableNode(
    Value=/Script/ControlRigSpline.RigUnit_ControlRigSplineFromPoints(SplineMode=BSpline, Points=RigVMVariableNode() {variable=points}.Value, SamplesPerSegment=RigVMVariableNode() {variable=local_samples}.Value) {is=struct RigVMUnitNode}.Spline
    ExecuteContext=@DISPATCH_RigVMDispatch_ArrayIterator.Completed
   ) {variable=spline}
   > SequenceAggregate = RigVMAggregateNode()
   RigVMFunction_ControlFlowBranch = &RigVMFunction_ControlFlowBranch(ExecuteContext=@SequenceAggregate.A, Condition=RigVMVariableNode() {variable=local_stretch}.Value)
   /Script/ControlRigSpline.RigUnit_FitChainToSplineCurveItemArray(
    Alignment=Front
    DebugSettings=(Scale=1,CurveColor=(R=1,G=1,B=0,A=1),SegmentsColor=(R=1,G=0,B=0,A=1),WorldOffset=(Rotation=(X=0,Y=0,Z=0,W=1),Translation=(X=0,Y=0,Z=0),Scale3D=(X=1,Y=1,Z=1)))
    Maximum=1
    PoleVectorPosition=(X=0,Y=0,Z=0)
    RotationEaseType=CubicEaseInOut
    Rotations=()
    SamplingPrecision=12
    SecondaryAxis=(X=0,Y=0,Z=0)
    Weight=1
    bPropagateToChildren=True
    ExecutePin=@RigVMFunction_ControlFlowBranch.False
    Spline=RigVMVariableNode() {variable=spline}.Value
    PrimaryAxis=RigVMVariableNode() {variable=local_primary_axis}.Value
    Items=RigVMVariableNode() {variable=local_bones}.Value
   ) {is=struct RigVMUnitNode}
   &RigUnit_FitChainToSplineCurveItemArray(
    Alignment=Stretched
    DebugSettings=(Scale=1,CurveColor=(R=1,G=1,B=0,A=1),SegmentsColor=(R=1,G=0,B=0,A=1),WorldOffset=(Rotation=(X=0,Y=0,Z=0,W=1),Translation=(X=0,Y=0,Z=0),Scale3D=(X=1,Y=1,Z=1)))
    Maximum=1.5
    PoleVectorPosition=(X=0,Y=0,Z=0)
    Rotations=()
    SamplingPrecision=12
    SecondaryAxis=(X=0,Y=0,Z=0)
    Weight=1
    bPropagateToChildren=True
    ExecutePin=@RigVMFunction_ControlFlowBranch.True
    Spline=RigVMVariableNode() {variable=spline}.Value
    PrimaryAxis=RigVMVariableNode() {variable=local_primary_axis}.Value
    Items=RigVMVariableNode() {variable=local_bones}.Value
   )
   n2 = &RigVMFunction_ControlFlowBranch(ExecuteContext=@SequenceAggregate.B, Condition=RigVMVariableNode() {variable=local_debug_draw}.Value)
   /Script/ControlRigSpline.RigUnit_DrawControlRigSpline(Color=(R=1,G=0,B=0,A=1), Detail=16, Thickness=1, ExecutePin=@n2.True, Spline=RigVMVariableNode() {variable=spline}.Value) {is=struct RigVMUnitNode}
   RigVMFunctionReturnNode(ExecuteContext=@SequenceAggregate.C)
   RigVMVariableNode() {variable=local_max stretch}
   RigVMVariableNode() {variable=local_min stretch}
  ]}
 New Function_ContainedGraph {nodes [
   n1 = RigVMFunctionEntryNode()
   > DISPATCH_RigVMDispatch_ArrayIterator = RigVMDispatchNode(Array=@n1.InBones) {template="DISPATCH_RigVMDispatch_ArrayIterator(in Array,out Element,out Index,out Count,out Ratio)"}
   DISPATCH_RigVMDispatch_ArrayGetAtIndex = RigVMDispatchNode(Index=@DISPATCH_RigVMDispatch_ArrayIterator.Index, Array=@n1.InControls) {template="DISPATCH_RigVMDispatch_ArrayGetAtIndex(in Array,in Index,out Element)"}
   > SnapTransform = RigVMFunctionReferenceNode(
    Active=true
    Initial=true
    Propagate to Children=false
    Space=GlobalSpace
    Target=@DISPATCH_RigVMDispatch_ArrayGetAtIndex.Element
    Item=@DISPATCH_RigVMDispatch_ArrayIterator.Element
   )
   RigVMFunctionReturnNode(ExecuteContext=@DISPATCH_RigVMDispatch_ArrayIterator.Completed)
  ]}
 RigVMFunctionLibrary {nodes [
   set_pose_func = RigVMCollapseNode(ExecuteContext=(), bind_pose=(HierarchyTopologyVersion=-1,PoseHash=-1), target_pose=(HierarchyTopologyVersion=-1,PoseHash=-1))
   simple_spline = RigVMCollapseNode(
    ExecuteContext=()
    bones=()
    controls=()
    debug draw spline=False
    max stretch=0
    min stretch=0
    primary axis=(X=1,Y=0,Z=0)
    samples=8
    stretch=False
   )
   snap_transform_array = RigVMCollapseNode(ExecuteContext=(), InBones=(), InControls=())
  ]}
 RigVMModel {nodes [
   RigUnit_BeginExecution = &RigUnit_BeginExecution()
   SequenceAggregate_1 = RigVMAggregateNode(ExecuteContext=@RigUnit_BeginExecution.ExecutePin)
   n1 = &RigUnit_SphericalPoseReader(
    ActiveRegionScaleFactors=(PositiveWidth=1,NegativeWidth=1,PositiveHeight=1.5,NegativeHeight=1.5)
    ActiveRegionSize=0.08
    Debug=(bDrawDebug=false,DebugScale=25,DebugSegments=20,DebugThickness=0.25)
    DriverAxis=(X=1,Y=0,Z=0)
    DriverItem=(Type=Bone,Name="head")
    FalloffRegionScaleFactors=(PositiveWidth=1,NegativeWidth=1,PositiveHeight=1,NegativeHeight=1)
    FalloffSize=0.4
    OptionalParentItem=(Type=Bone,Name="spine_04")
    RotationOffset=(X=-90,Y=0,Z=-5)
    ExecutePin=@SequenceAggregate_1.A
   )
   > RigVMVariableNode(Value=@n1.OutputParam) {variable=spr_head_back}
   > n2 = &RigUnit_SphericalPoseReader(
    ActiveRegionScaleFactors=(PositiveWidth=1,NegativeWidth=1,PositiveHeight=1,NegativeHeight=1)
    ActiveRegionSize=0.05
    Debug=(bDrawDebug=false,DebugScale=25,DebugSegments=20,DebugThickness=0.25)
    DriverAxis=(X=1,Y=0,Z=0)
    DriverItem=(Type=Bone,Name="head")
    FalloffRegionScaleFactors=(PositiveWidth=1,NegativeWidth=1,PositiveHeight=1,NegativeHeight=1)
    FalloffSize=0.3
    OptionalParentItem=(Type=Bone,Name="spine_04")
    RotationOffset=(X=-35,Y=-25,Z=20)
   )
   > RigVMVariableNode(Value=@n2.OutputParam) {variable=spr_head_right}
   > n3 = &RigUnit_SphericalPoseReader(
    ActiveRegionScaleFactors=(PositiveWidth=1,NegativeWidth=1,PositiveHeight=1,NegativeHeight=1)
    ActiveRegionSize=0.1
    Debug=(bDrawDebug=false,DebugScale=25,DebugSegments=20,DebugThickness=0.25)
    DriverAxis=(X=1,Y=0,Z=0)
    DriverItem=(Type=Bone,Name="head")
    FalloffRegionScaleFactors=(PositiveWidth=1,NegativeWidth=1,PositiveHeight=1,NegativeHeight=1)
    FalloffSize=0.4
    OptionalParentItem=(Type=Bone)
    RotationOffset=(X=180,Y=0,Z=0)
   )
   > RigVMVariableNode(Value=@n3.OutputParam) {variable=spr_head_left}
   > n4 = &RigUnit_SphericalPoseReader(
    ActiveRegionScaleFactors=(PositiveWidth=5,NegativeWidth=1,PositiveHeight=3,NegativeHeight=5)
    ActiveRegionSize=0.1
    Debug=(bDrawDebug=false,DebugScale=25,DebugSegments=20,DebugThickness=0.25)
    DriverAxis=(X=-1,Y=0,Z=0)
    DriverItem=(Type=Bone,Name="upperarm_r")
    FalloffRegionScaleFactors=(PositiveWidth=1,NegativeWidth=1,PositiveHeight=1,NegativeHeight=1)
    FalloffSize=0.5
    OptionalParentItem=(Type=Bone,Name="clavicle_r")
    RotationOffset=(X=-120,Y=20,Z=0)
   )
   > RigVMVariableNode(Value=@n4.OutputParam) {variable=spr_arm_raise_upper_right}
   > n5 = &RigUnit_SphericalPoseReader(
    ActiveRegionScaleFactors=(PositiveWidth=1,NegativeWidth=1,PositiveHeight=1,NegativeHeight=1)
    ActiveRegionSize=0.05
    Debug=(bDrawDebug=false,DebugScale=25,DebugSegments=20,DebugThickness=0.25)
    DriverAxis=(X=-1,Y=0,Z=0)
    DriverItem=(Type=Bone,Name="clavicle_r")
    FalloffRegionScaleFactors=(PositiveWidth=1,NegativeWidth=1,PositiveHeight=1,NegativeHeight=1)
    FalloffSize=0.15
    OptionalParentItem=(Type=Bone,Name="spine_05")
    RotationOffset=(X=-5,Y=130,Z=0)
   )
   > n6 = RigVMVariableNode(Value=@n5.OutputParam) {variable=spr_clavicle_raise_right}
   RigVMFunction_ControlFlowBranch = &RigVMFunction_ControlFlowBranch(ExecuteContext=@n6.ExecuteContext, Condition=RigVMVariableNode() {variable=IsJacketVariant}.Value)
   n7 = &RigUnit_SphericalPoseReader(
    ActiveRegionScaleFactors=(PositiveWidth=5,NegativeWidth=1,PositiveHeight=2.5,NegativeHeight=2.5)
    ActiveRegionSize=0.1
    Debug=(bDrawDebug=false,DebugScale=25,DebugSegments=20,DebugThickness=0.25)
    DriverAxis=(X=-1,Y=0,Z=0)
    DriverItem=(Type=Bone,Name="upperarm_r")
    FalloffRegionScaleFactors=(PositiveWidth=1,NegativeWidth=0.75,PositiveHeight=1,NegativeHeight=1)
    FalloffSize=0.5
    FlipWidthScaling=false
    OptionalParentItem=(Type=Bone,Name="spine_04")
    RotationOffset=(X=-60,Y=30,Z=20)
    ExecutePin=@RigVMFunction_ControlFlowBranch.False
   )
   > RigVMVariableNode(Value=@n7.OutputParam) {variable=spr_arm_raise_lower_right}
   n8 = &RigUnit_SphericalPoseReader(
    ActiveRegionScaleFactors=(PositiveWidth=5,NegativeWidth=1,PositiveHeight=1.2,NegativeHeight=5)
    ActiveRegionSize=0.1
    Debug=(bDrawDebug=false,DebugScale=25,DebugSegments=20,DebugThickness=0.25)
    DriverAxis=(X=-1,Y=0,Z=0)
    DriverItem=(Type=Bone,Name="upperarm_r")
    FalloffRegionScaleFactors=(PositiveWidth=1,NegativeWidth=1,PositiveHeight=1,NegativeHeight=1)
    FalloffSize=0.5
    OptionalParentItem=(Type=Bone,Name="clavicle_r")
    RotationOffset=(X=-90,Y=20,Z=0)
    ExecutePin=@RigVMFunction_ControlFlowBranch.True
   )
   > RigVMVariableNode(Value=@n8.OutputParam) {variable=spr_arm_raise_lower_right}
   > n9 = &RigUnit_SphericalPoseReader(
    ActiveRegionScaleFactors=(PositiveWidth=5,NegativeWidth=1,PositiveHeight=1,NegativeHeight=5)
    ActiveRegionSize=0.1
    Debug=(bDrawDebug=false,DebugScale=25,DebugSegments=20,DebugThickness=0.25)
    DriverAxis=(X=-1,Y=0,Z=0)
    DriverItem=(Type=Bone,Name="upperarm_r")
    FalloffRegionScaleFactors=(PositiveWidth=1,NegativeWidth=1,PositiveHeight=1,NegativeHeight=1)
    FalloffSize=0.5
    OptionalParentItem=(Type=Bone,Name="clavicle_r")
    RotationOffset=(X=-105,Y=20,Z=0)
   )
   > RigVMVariableNode(Value=@n9.OutputParam) {variable=spr_arm_raise_jacket_mid}
   > n10 = &RigUnit_SphericalPoseReader(
    ActiveRegionScaleFactors=(PositiveWidth=5,NegativeWidth=1,PositiveHeight=1,NegativeHeight=5)
    ActiveRegionSize=0.1
    Debug=(bDrawDebug=false,DebugScale=25,DebugSegments=20,DebugThickness=0.25)
    DriverAxis=(X=-1,Y=0,Z=0)
    DriverItem=(Type=Bone,Name="upperarm_r")
    FalloffRegionScaleFactors=(PositiveWidth=1,NegativeWidth=1,PositiveHeight=1,NegativeHeight=1)
    FalloffSize=0.5
    OptionalParentItem=(Type=Bone,Name="clavicle_r")
    RotationOffset=(X=-120,Y=0,Z=30)
   )
   > RigVMVariableNode(Value=@n10.OutputParam) {variable=spr_arm_raise_jacket_upper}
   > n11 = &RigUnit_SphericalPoseReader(
    ActiveRegionScaleFactors=(PositiveWidth=1,NegativeWidth=1,PositiveHeight=1,NegativeHeight=1)
    ActiveRegionSize=0.02
    Debug=(bDrawDebug=true,DebugScale=25,DebugSegments=20,DebugThickness=0.25)
    DriverAxis=(X=-1,Y=0,Z=0)
    DriverItem=(Type=Bone,Name="clavicle_r")
    FalloffRegionScaleFactors=(PositiveWidth=1,NegativeWidth=1,PositiveHeight=1,NegativeHeight=1)
    FalloffSize=0.15
    OptionalParentItem=(Type=Bone,Name="spine_05")
    RotationOffset=(X=-130,Y=0,Z=-75)
   )
   > RigVMVariableNode(Value=@n11.OutputParam) {variable=spr_clavicle_up_and_forward}
   &RigUnit_ParentConstraint(
    AdvancedSettings=(RotationOrderForFilter=XZY)
    Child=(Type=Null,Name="hair_r_front_04_ctrl_null")
    Parents="((Item=(Type=Bone,Name=\"clavicle_r\"),Weight=0.005),(Item=(Type=Bone,Name=\"spine_04\"),Weight=0.0283),(Item=(Type=Bone,Name=\"head\"),Weight=0.1079),(Item=(Type=Bone,Name=\"neck_01\"),Weight=0.1809),(Item=(Type=Bone,Name=\"neck_02\"),Weight=0.271),(Item=(Type=Bone ...(truncated)"
    Weight=1
    ExecutePin=@SequenceAggregate_1.B
   )
   > &RigUnit_ParentConstraint(
    AdvancedSettings=(RotationOrderForFilter=XZY)
    Child=(Type=Null,Name="hair_r_front_05_ctrl_null")
    Parents="((Item=(Type=Bone,Name=\"deform_upperarm_up_r\"),Weight=0.0196),(Item=(Type=Bone,Name=\"deform_pec_r\"),Weight=0.0686),(Item=(Type=Bone,Name=\"upperarm_twist_01_r\"),Weight=0.0798),(Item=(Type=Bone,Name=\"deform_upperarm_fr_r\"),Weight=0.0899),(Item=(Type=Bone,Nam ...(truncated)"
    Weight=1
   )
   > &RigUnit_ParentConstraint(
    AdvancedSettings=(RotationOrderForFilter=XZY)
    Child=(Type=Null,Name="hair_r_front_06_ctrl_null")
    Parents="((Item=(Type=Bone,Name=\"spine_02\"),Weight=0.0001),(Item=(Type=Bone,Name=\"spine_03\"),Weight=0.0002),(Item=(Type=Bone,Name=\"deform_upperarm_up_r\"),Weight=0.0015),(Item=(Type=Bone,Name=\"deform_upperarm_fr_r\"),Weight=0.0831),(Item=(Type=Bone,Name=\"upperarm_twi ...(truncated)"
    Weight=1
   )
   > &RigUnit_ParentConstraint(
    AdvancedSettings=(RotationOrderForFilter=XZY)
    Child=(Type=Null,Name="hair_r_front_07_ctrl_null")
    Parents="((Item=(Type=Bone,Name=\"spine_03\"),Weight=0.004),(Item=(Type=Bone,Name=\"spine_02\"),Weight=0.0042),(Item=(Type=Bone,Name=\"deform_upperarm_fr_r\"),Weight=0.0171),(Item=(Type=Bone,Name=\"upperarm_twist_01_r\"),Weight=0.0356),(Item=(Type=Bone,Name=\"clavicle_r\"),W ...(truncated)"
    Weight=1
   )
   set_pose_func_2 = RigVMFunctionReferenceNode(
    ExecuteContext=@SequenceAggregate_1.C
    pose_alpha=RigVMVariableNode() {variable=spr_arm_raise_lower_right}.Value
    target_pose=RigVMVariableNode() {variable=r_arm_fwd_lower_ctrls_pose}.Value
    bind_pose=RigVMVariableNode() {variable=front_hair_ctrl_bind}.Value
   )
   > set_pose_func_2_1 = RigVMFunctionReferenceNode(
    bind_pose=RigVMVariableNode() {variable=front_hair_ctrl_bind}.Value
    pose_alpha=RigVMVariableNode() {variable=spr_arm_raise_upper_right}.Value
    target_pose=RigVMVariableNode() {variable=r_clav_fwd_upper_ctrls_pose}.Value
   )
   n12 = RigVMVariableNode() {variable=hair_aim_ctrls_bind}
   > set_pose_func_2_1_4 = RigVMFunctionReferenceNode(
    bind_pose=@n12.Value
    target_pose=RigVMVariableNode() {variable=hair_aim_ctrls_head_back_pose}.Value
    pose_alpha=RigVMVariableNode() {variable=spr_head_back}.Value
   )
   > set_pose_func_2_1_4_1 = RigVMFunctionReferenceNode(
    target_pose=RigVMVariableNode() {variable=HierarchyGetPoseItemArray_5_1_1_1_Pose}.Value
    pose_alpha=RigVMVariableNode() {variable=spr_head_right}.Value
    bind_pose=@n12.Value
   )
   n13 = &RigVMFunction_ControlFlowBranch(Condition=RigVMVariableNode() {variable=IsJacketVariant}.Value, ExecuteContext=@set_pose_func_2_1_4_1.ExecuteContext)
   set_pose_func_2_1_4_1_1 = RigVMFunctionReferenceNode(
    ExecuteContext=@n13.False
    bind_pose=RigVMVariableNode() {variable=front_hair_ctrl_bind}.Value
    target_pose=RigVMVariableNode() {variable=front_hair_arm_raise_no_jacket_pose}.Value
    pose_alpha=RigVMVariableNode() {variable=spr_arm_raise_lower_right}.Value
   )
   set_pose_func_2_1_4_1_1_1_1_1_1 = RigVMFunctionReferenceNode(
    ExecuteContext=@n13.True
    pose_alpha=RigVMVariableNode() {variable=spr_clavicle_up_and_forward}.Value
    target_pose=RigVMVariableNode() {variable=front_hair_clavicle_raised_and_fwd_pose}.Value
    bind_pose=RigVMVariableNode() {variable=front_hair_ctrl_bind}.Value
   )
   n14 = /Script/ControlRig.RigUnit_HierarchyGetPoseItemArray(
    ElementType=All
    Initial=false
    ItemsToGet="((Type=Control,Name=\"hair_r_front_01_ctrl\"),(Type=Control,Name=\"hair_r_front_02_ctrl\"),(Type=Control,Name=\"hair_r_front_03_ctrl\"),(Type=Control,Name=\"hair_r_front_04_ctrl\"),(Type=Control,Name=\"hair_r_front_05_ctrl\"),(Type=Control,Name=\"hair_r_front_06_ctrl ...(truncated)"
   ) {is=struct RigVMUnitNode}
   > set_pose_func_2_1_4_1_1_1 = RigVMFunctionReferenceNode(
    bind_pose=@n14.Pose
    target_pose=RigVMVariableNode() {variable=front_hair_arm_raise_jacket_pose}.Value
    pose_alpha=RigVMVariableNode() {variable=spr_arm_raise_lower_right}.Value
   )
   n15 = &RigUnit_HierarchyGetPoseItemArray(
    ElementType=All
    Initial=false
    ItemsToGet="((Type=Control,Name=\"hair_r_front_01_ctrl\"),(Type=Control,Name=\"hair_r_front_02_ctrl\"),(Type=Control,Name=\"hair_r_front_03_ctrl\"),(Type=Control,Name=\"hair_r_front_04_ctrl\"),(Type=Control,Name=\"hair_r_front_05_ctrl\"),(Type=Control,Name=\"hair_r_front_06_ctrl ...(truncated)"
   )
   > set_pose_func_2_1_4_1_1_1_1 = RigVMFunctionReferenceNode(
    bind_pose=@n15.Pose
    target_pose=RigVMVariableNode() {variable=front_hair_arm_raise_jacket_mid_pose}.Value
    pose_alpha=RigVMVariableNode() {variable=spr_arm_raise_jacket_mid}.Value
   )
   n16 = &RigUnit_HierarchyGetPoseItemArray(
    ElementType=All
    Initial=false
    ItemsToGet="((Type=Control,Name=\"hair_r_front_01_ctrl\"),(Type=Control,Name=\"hair_r_front_02_ctrl\"),(Type=Control,Name=\"hair_r_front_03_ctrl\"),(Type=Control,Name=\"hair_r_front_04_ctrl\"),(Type=Control,Name=\"hair_r_front_05_ctrl\"),(Type=Control,Name=\"hair_r_front_06_ctrl ...(truncated)"
   )
   > set_pose_func_2_1_4_1_1_1_1_1 = RigVMFunctionReferenceNode(
    bind_pose=@n16.Pose
    target_pose=RigVMVariableNode() {variable=front_hair_arm_Raise_jacket_upper_pose}.Value
    pose_alpha=RigVMVariableNode() {variable=spr_arm_raise_jacket_upper}.Value
   )
   SplineIK_2_2_2_1 = RigVMFunctionReferenceNode(
    Primary Axis=(X=1,Y=0,Z=0)
    Samples Per Segment=8
    Secondary Spline Direction=(X=0,Y=2,Z=0)
    Spline Mode=BSpline
    Stretch=false
    Up Axis=(X=0,Y=1,Z=0)
    debug=true
    Bones=RigVMVariableNode() {variable=front_hair_bones}.Value
    Controls=RigVMVariableNode() {variable=front_r_hair_ctrls}.Value
    ExecuteContext=@SequenceAggregate_1.E
   )
   &RigUnit_ModifyTransforms(
    ItemToModify="((Item=(Type=Bone,Name=\"hair_r_side_01_root\"),Transform=(Rotation=(X=0,Y=0,Z=0,W=1),Translation=(X=-2,Y=-5,Z=0),Scale3D=(X=1,Y=1,Z=1))),(Item=(Type=Bone,Name=\"hair_l_side_01_root\"),Transform=(Rotation=(X=0,Y=0,Z=0,W=1),Translation=(X=-2,Y=-5,Z=0),Scale3D=( ...(truncated)"
    Weight=RigVMVariableNode() {variable=spr_head_back}.Value
    ExecutePin=@SequenceAggregate_1.D
   )
   > &RigUnit_ModifyTransforms(
    ItemToModify="((Item=(Type=Bone,Name=\"hair_r_side_01_root\"),Transform=(Rotation=(X=0,Y=0,Z=0,W=1),Translation=(X=-5,Y=-3,Z=2),Scale3D=(X=1,Y=1,Z=1))),(Item=(Type=Bone,Name=\"hair_l_side_01_root\"),Transform=(Rotation=(X=0,Y=0,Z=0,W=1),Translation=(X=0,Y=0,Z=0),Scale3D=(X= ...(truncated)"
    Weight=RigVMVariableNode() {variable=spr_head_right}.Value
   )
   > &RigUnit_ModifyTransforms(
    ItemToModify="((Item=(Type=Bone,Name=\"hair_r_side_01_root\"),Transform=(Rotation=(X=0,Y=0,Z=0,W=1),Translation=(X=-2,Y=-5,Z=0),Scale3D=(X=1,Y=1,Z=1))),(Item=(Type=Bone,Name=\"hair_rear_01_root\"),Transform=(Rotation=(X=0,Y=0,Z=0,W=1),Translation=(X=0,Y=-3,Z=0),Scale3D=(X=1 ...(truncated)"
    Weight=RigVMVariableNode() {variable=spr_clavicle_raise_right}.Value
   )
   > &RigUnit_AimConstraintLocalSpaceOffset(
    AdvancedSettings=(DebugSettings=(Scale=10,WorldOffset=(Rotation=(X=0,Y=0,Z=0,W=1),Translation=(X=0,Y=0,Z=0),Scale3D=(X=1,Y=1,Z=1))),RotationOrderForFilter=XZY)
    Child=(Type=Bone,Name="hair_rear_01_root")
    Filter=(bX=True,bY=True,bZ=True)
    Parents=((Item=(Type=Control,Name="hair_rear_06_ctrl_2"),Weight=1))
    Weight=1
    WorldUp=(Target=(X=0,Y=0,Z=1),Kind=Location,Space=(Type=Bone,Name="spine_04"))
   )
   > &RigUnit_AimConstraintLocalSpaceOffset(
    AdvancedSettings=(DebugSettings=(Scale=10,WorldOffset=(Rotation=(X=0,Y=0,Z=0,W=1),Translation=(X=0,Y=0,Z=0),Scale3D=(X=1,Y=1,Z=1))),RotationOrderForFilter=XZY)
    Child=(Type=Bone,Name="hair_l_side_01_root")
    Filter=(bX=True,bY=True,bZ=True)
    Parents=((Item=(Type=Control,Name="hair_l_side_06_ctrl_2"),Weight=1))
    Weight=1
    WorldUp=(Target=(X=0,Y=0,Z=1),Kind=Location,Space=(Type=Bone,Name="spine_04"))
   )
   > &RigUnit_AimConstraintLocalSpaceOffset(
    AdvancedSettings=(DebugSettings=(Scale=10,WorldOffset=(Rotation=(X=0,Y=0,Z=0,W=1),Translation=(X=0,Y=0,Z=0),Scale3D=(X=1,Y=1,Z=1))),RotationOrderForFilter=XZY)
    Child=(Type=Bone,Name="hair_r_side_01_root")
    Filter=(bX=True,bY=True,bZ=True)
    Parents=((Item=(Type=Control,Name="hair_r_side_06_ctrl_2"),Weight=1))
    Weight=1
    WorldUp=(Target=(X=0,Y=0,Z=1),Kind=Location,Space=(Type=Bone,Name="spine_04"))
   )
   n17 = &RigUnit_SphericalPoseReader(
    ActiveRegionScaleFactors=(PositiveWidth=1,NegativeWidth=1,PositiveHeight=1,NegativeHeight=1)
    ActiveRegionSize=0.1
    Debug=(bDrawDebug=false,DebugScale=25,DebugSegments=20,DebugThickness=0.25)
    DriverAxis=(X=1,Y=0,Z=0)
    DriverItem=(Type=Bone,Name="clavicle_l")
    ExecutePin=()
    FalloffRegionScaleFactors=(PositiveWidth=1,NegativeWidth=1,PositiveHeight=1,NegativeHeight=1)
    FalloffSize=0.2
    OptionalParentItem=(Type=Bone)
    RotationOffset=(X=46,Y=0,Z=-83)
   )
   > RigVMVariableNode(Value=@n17.OutputParam) {variable=spr_clavicle_raise_left}
   simple_spline = RigVMFunctionReferenceNode(
    ExecuteContext=()
    bones=()
    controls=()
    debug draw spline=False
    max stretch=0
    min stretch=0
    primary axis=(X=1,Y=0,Z=0)
    samples=8
    stretch=False
   )
  ]}
 RigVMModel Rig Graph {nodes [
   n1 = RigVMDispatchNode(ExecuteContext=(), Array=RigVMVariableNode() {variable=front_r_hair_ctrls}.Value) {template="DISPATCH_RigVMDispatch_ArrayIterator(in Array,out Element,out Index,out Count,out Ratio)"}
   > &RigUnit_SetTransform(
    Space=LocalSpace
    Weight=1
    bInitial=False
    Item=@n1.Element
    Value=/Script/RigVM.RigVMFunction_MathTransformMake(Rotation=(X=0,Y=0,Z=0,W=1), Scale=(X=1,Y=1,Z=1), Translation=(X=0,Y=0,Z=0)) {is=struct RigVMUnitNode}.Result
   )
   n2 = &RigUnit_PrepareForExecution()
   SequenceAggregate = RigVMAggregateNode(ExecuteContext=@n2.ExecutePin)
   n3 = &RigUnit_ItemArray(Items=((Type=Bone,Name="hair_r_front_01_root"),(Type=Bone,Name="hair_r_front_02"),(Type=Bone,Name="hair_r_front_03"),(Type=Bone,Name="hair_r_front_04"),(Type=Bone,Name="hair_r_front_05"),(Type=Bone,Name="hair_r_front_06"),(Type=Bone,Name="hair_r_front_07")))
   RigVMVariableNode(Value=@n3.Items, ExecuteContext=@SequenceAggregate.A) {variable=front_hair_bones}
   n4 = &RigUnit_ItemArray(Items="((Type=Control,Name=\"hair_r_front_01_ctrl\"),(Type=Control,Name=\"hair_r_front_02_ctrl\"),(Type=Control,Name=\"hair_r_front_03_ctrl\"),(Type=Control,Name=\"hair_r_front_04_ctrl\"),(Type=Control,Name=\"hair_r_front_05_ctrl\"),(Type=Control,Name=\"hair_r_front_06_ctrl ...(truncated)")
   > RigVMVariableNode(Value=@n4.Items) {variable=front_r_hair_ctrls}
   > snap_transform_array_3 = RigVMFunctionReferenceNode(InBones=RigVMVariableNode() {variable=front_hair_bones}.Value, InControls=RigVMVariableNode() {variable=front_r_hair_ctrls}.Value)
   n5 = &RigUnit_HierarchyGetPoseItemArray(
    ElementType=All
    Initial=false
    ItemsToGet=((Type=Control,Name="hair_l_side_06_ctrl_2"),(Type=Control,Name="hair_r_side_06_ctrl_2"),(Type=Control,Name="hair_rear_06_ctrl_2"))
   )
   RigVMVariableNode(Value=@n5.Pose, ExecuteContext=@SequenceAggregate.B) {variable=hair_aim_ctrls_bind}
   > RigVMVariableNode(Value=&RigUnit_HierarchyGetPoseItemArray(ElementType=All, Initial=false, ItemsToGet=RigVMVariableNode() {variable=front_r_hair_ctrls}.Value).Pose) {variable=front_hair_ctrl_bind}
   &RigUnit_ModifyTransforms(
    ItemToModify="((Item=(Type=Control,Name=\"hair_l_side_06_ctrl_2\"),Transform=(Rotation=(X=0,Y=0,Z=0,W=1),Translation=(X=0,Y=0,Z=10),Scale3D=(X=1,Y=1,Z=1))),(Item=(Type=Control,Name=\"hair_r_side_06_ctrl_2\"),Transform=(Rotation=(X=0,Y=0,Z=0,W=1),Translation=(X=0,Y=10,Z=0),S ...(truncated)"
    Weight=1
    ExecutePin=@SequenceAggregate.C
   )
   n6 = &RigUnit_HierarchyGetPoseItemArray(
    ElementType=All
    Initial=false
    ItemsToGet=((Type=Control,Name="hair_l_side_06_ctrl_2"),(Type=Control,Name="hair_r_side_06_ctrl_2"),(Type=Control,Name="hair_rear_06_ctrl_2"))
   )
   > RigVMVariableNode(Value=@n6.Pose) {variable=hair_aim_ctrls_head_back_pose}
   > &RigUnit_ModifyTransforms(
    ItemToModify="((Item=(Type=Control,Name=\"hair_l_side_06_ctrl_2\"),Transform=(Rotation=(X=0,Y=0,Z=0,W=1),Translation=(X=0,Y=0,Z=0),Scale3D=(X=1,Y=1,Z=1))),(Item=(Type=Control,Name=\"hair_r_side_06_ctrl_2\"),Transform=(Rotation=(X=0,Y=0,Z=0,W=1),Translation=(X=0,Y=-8,Z=0),Sc ...(truncated)"
    Weight=1
   )
   n7 = &RigUnit_HierarchyGetPoseItemArray(
    ElementType=All
    Initial=false
    ItemsToGet=((Type=Control,Name="hair_l_side_06_ctrl_2"),(Type=Control,Name="hair_r_side_06_ctrl_2"),(Type=Control,Name="hair_rear_06_ctrl_2"))
   )
   > RigVMVariableNode(Value=@n7.Pose) {variable=HierarchyGetPoseItemArray_5_1_1_1_Pose}
   &RigUnit_ModifyTransforms(
    ItemToModify="((Item=(Type=Control,Name=\"hair_r_front_01_ctrl\"),Transform=(Rotation=(X=0,Y=0,Z=0,W=1),Translation=(X=0,Y=0,Z=0),Scale3D=(X=1,Y=1,Z=1))),(Item=(Type=Control,Name=\"hair_r_front_02_ctrl\"),Transform=(Rotation=(X=0,Y=0,Z=0,W=1),Translation=(X=0,Y=0,Z=0),Scale ...(truncated)"
    Weight=1
    ExecutePin=@SequenceAggregate.D
   )
   n8 = &RigUnit_HierarchyGetPoseItemArray(
    ElementType=All
    Initial=false
    ItemsToGet="((Type=Control,Name=\"hair_r_front_01_ctrl\"),(Type=Control,Name=\"hair_r_front_02_ctrl\"),(Type=Control,Name=\"hair_r_front_03_ctrl\"),(Type=Control,Name=\"hair_r_front_04_ctrl\"),(Type=Control,Name=\"hair_r_front_05_ctrl\"),(Type=Control,Name=\"hair_r_front_06_ctrl ...(truncated)"
   )
   > RigVMVariableNode(Value=@n8.Pose) {variable=front_hair_arm_raise_no_jacket_pose}
   > DISPATCH_RigVMDispatch_ArrayIterator = RigVMDispatchNode(Array=RigVMVariableNode() {variable=front_r_hair_ctrls}.Value) {template="DISPATCH_RigVMDispatch_ArrayIterator(in Array,out Element,out Index,out Count,out Ratio)"}
   RigVMFunction_MathTransformMake = &RigVMFunction_MathTransformMake(Rotation=(X=0,Y=0,Z=0,W=1), Scale=(X=1,Y=1,Z=1), Translation=(X=0,Y=0,Z=0))
   > &RigUnit_SetTransform(
    Space=LocalSpace
    Weight=1
    bInitial=False
    Item=@DISPATCH_RigVMDispatch_ArrayIterator.Element
    Value=@RigVMFunction_MathTransformMake.Result
   )
   &RigUnit_ModifyTransforms(
    ItemToModify="((Item=(Type=Control,Name=\"hair_r_front_01_ctrl\"),Transform=(Rotation=(X=0,Y=0,Z=0,W=1),Translation=(X=0,Y=0,Z=0),Scale3D=(X=1,Y=1,Z=1))),(Item=(Type=Control,Name=\"hair_r_front_02_ctrl\"),Transform=(Rotation=(X=0,Y=0,Z=0,W=1),Translation=(X=0,Y=0,Z=0),Scale ...(truncated)"
    Weight=1
    ExecutePin=@DISPATCH_RigVMDispatch_ArrayIterator.Completed
   )
   n9 = &RigUnit_HierarchyGetPoseItemArray(
    ElementType=All
    Initial=false
    ItemsToGet="((Type=Control,Name=\"hair_r_front_01_ctrl\"),(Type=Control,Name=\"hair_r_front_02_ctrl\"),(Type=Control,Name=\"hair_r_front_03_ctrl\"),(Type=Control,Name=\"hair_r_front_04_ctrl\"),(Type=Control,Name=\"hair_r_front_05_ctrl\"),(Type=Control,Name=\"hair_r_front_06_ctrl ...(truncated)"
   )
   > RigVMVariableNode(Value=@n9.Pose) {variable=front_hair_arm_raise_jacket_pose}
   > &RigUnit_ModifyTransforms(
    ItemToModify="((Item=(Type=Control,Name=\"hair_r_front_01_ctrl\"),Transform=(Rotation=(X=0,Y=0,Z=0,W=1),Translation=(X=0,Y=0,Z=0),Scale3D=(X=1,Y=1,Z=1))),(Item=(Type=Control,Name=\"hair_r_front_02_ctrl\"),Transform=(Rotation=(X=0,Y=0,Z=0,W=1),Translation=(X=0,Y=0,Z=0),Scale ...(truncated)"
    Weight=1
   )
   n10 = &RigUnit_HierarchyGetPoseItemArray(
    ElementType=All
    Initial=false
    ItemsToGet="((Type=Control,Name=\"hair_r_front_01_ctrl\"),(Type=Control,Name=\"hair_r_front_02_ctrl\"),(Type=Control,Name=\"hair_r_front_03_ctrl\"),(Type=Control,Name=\"hair_r_front_04_ctrl\"),(Type=Control,Name=\"hair_r_front_05_ctrl\"),(Type=Control,Name=\"hair_r_front_06_ctrl ...(truncated)"
   )
   > RigVMVariableNode(Value=@n10.Pose) {variable=front_hair_arm_raise_jacket_mid_pose}
   > &RigUnit_ModifyTransforms(
    ItemToModify="((Item=(Type=Control,Name=\"hair_r_front_01_ctrl\"),Transform=(Rotation=(X=0,Y=0,Z=0,W=1),Translation=(X=0,Y=0,Z=0),Scale3D=(X=1,Y=1,Z=1))),(Item=(Type=Control,Name=\"hair_r_front_02_ctrl\"),Transform=(Rotation=(X=0,Y=0,Z=0,W=1),Translation=(X=0,Y=0,Z=0),Scale ...(truncated)"
    Weight=1
   )
   n11 = &RigUnit_HierarchyGetPoseItemArray(
    ElementType=All
    Initial=false
    ItemsToGet="((Type=Control,Name=\"hair_r_front_01_ctrl\"),(Type=Control,Name=\"hair_r_front_02_ctrl\"),(Type=Control,Name=\"hair_r_front_03_ctrl\"),(Type=Control,Name=\"hair_r_front_04_ctrl\"),(Type=Control,Name=\"hair_r_front_05_ctrl\"),(Type=Control,Name=\"hair_r_front_06_ctrl ...(truncated)"
   )
   > RigVMVariableNode(Value=@n11.Pose) {variable=front_hair_arm_Raise_jacket_upper_pose}
   > n12 = RigVMDispatchNode(Array=()) {template="DISPATCH_RigVMDispatch_ArrayIterator(in Array,out Element,out Index,out Count,out Ratio)"}
   > &RigUnit_SetTransform(
    Space=LocalSpace
    Weight=1
    bInitial=False
    Item=@n12.Element
    Value=&RigVMFunction_MathTransformMake(Rotation=(X=0,Y=0,Z=0,W=1), Scale=(X=1,Y=1,Z=1), Translation=(X=0,Y=0,Z=0)).Result
   )
   &RigUnit_ModifyTransforms(
    ItemToModify="((Item=(Type=Control,Name=\"hair_r_front_01_ctrl\"),Transform=(Rotation=(X=0,Y=0,Z=0,W=1),Translation=(X=0,Y=0,Z=0),Scale3D=(X=1,Y=1,Z=1))),(Item=(Type=Control,Name=\"hair_r_front_02_ctrl\"),Transform=(Rotation=(X=0,Y=0,Z=0,W=1),Translation=(X=0,Y=0,Z=0),Scale ...(truncated)"
    Weight=1
    ExecutePin=@n12.Completed
   )
   n13 = &RigUnit_HierarchyGetPoseItemArray(
    ElementType=All
    Initial=false
    ItemsToGet="((Type=Control,Name=\"hair_r_front_01_ctrl\"),(Type=Control,Name=\"hair_r_front_02_ctrl\"),(Type=Control,Name=\"hair_r_front_03_ctrl\"),(Type=Control,Name=\"hair_r_front_04_ctrl\"),(Type=Control,Name=\"hair_r_front_05_ctrl\"),(Type=Control,Name=\"hair_r_front_06_ctrl ...(truncated)"
   )
   > RigVMVariableNode(Value=@n13.Pose) {variable=front_hair_clavicle_raised_and_fwd_pose}
   snap_transform_array = RigVMFunctionReferenceNode(
    ExecuteContext=()
    InControls=RigVMVariableNode() {variable=rear_hair_ctrls}.Value
    InBones=RigVMVariableNode() {variable=rear_hair_bones}.Value
   )
   > snap_transform_array_1 = RigVMFunctionReferenceNode(InBones=RigVMVariableNode() {variable=side_l_hair_bones}.Value, InControls=RigVMVariableNode() {variable=side_l_hair_ctrls}.Value)
   > snap_transform_array_2 = RigVMFunctionReferenceNode(InBones=RigVMVariableNode() {variable=side_r_hair_bones}.Value, InControls=RigVMVariableNode() {variable=side_r_hair_ctrls}.Value)
   RigVMFunction_ControlFlowBranch = &RigVMFunction_ControlFlowBranch(ExecuteContext=(), Condition=RigVMVariableNode() {variable=IsJacketVariant}.Value)
  ]}
 SequenceAggregate_1_1_ContainedGraph {nodes [
   n1 = RigVMFunctionEntryNode()
   SequenceAggregate_1 = &RigVMFunction_Sequence(ExecuteContext=@n1.ExecuteContext)
   SequenceAggregate_1_1 = &RigVMFunction_Sequence(ExecuteContext=@SequenceAggregate_1.B)
   SequenceAggregate_1_2 = &RigVMFunction_Sequence(ExecuteContext=@SequenceAggregate_1_1.B)
   SequenceAggregate_1_3 = &RigVMFunction_Sequence(ExecuteContext=@SequenceAggregate_1_2.B)
   RigVMFunctionReturnNode(A=@SequenceAggregate_1.A, B=@SequenceAggregate_1_1.A, C=@SequenceAggregate_1_2.A, D=@SequenceAggregate_1_3.A, E=@SequenceAggregate_1_3.B)
  ]}
 SequenceAggregate_1_ContainedGraph {nodes [
   n1 = RigVMFunctionEntryNode()
   SequenceAggregate = &RigVMFunction_Sequence(ExecuteContext=@n1.ExecuteContext)
   SequenceAggregate_1 = &RigVMFunction_Sequence(ExecuteContext=@SequenceAggregate.B)
   SequenceAggregate_2 = &RigVMFunction_Sequence(ExecuteContext=@SequenceAggregate_1.B)
   RigVMFunctionReturnNode(A=@SequenceAggregate.A, B=@SequenceAggregate_1.A, C=@SequenceAggregate_2.A, D=@SequenceAggregate_2.B)
  ]}
 SequenceAggregate_1_ContainedGraph {nodes [
   n1 = RigVMFunctionEntryNode()
   SequenceAggregate = &RigVMFunction_Sequence(ExecuteContext=@n1.ExecuteContext)
   SequenceAggregate_1 = &RigVMFunction_Sequence(ExecuteContext=@SequenceAggregate.B)
   RigVMFunctionReturnNode(A=@SequenceAggregate.A, B=@SequenceAggregate_1.A, C=@SequenceAggregate_1.B)
  ]}
 set_pose_func_ContainedGraph {nodes [
   n1 = RigVMFunctionEntryNode()
   SequenceAggregate = &RigVMFunction_Sequence(ExecuteContext=@n1.ExecuteContext)
   Subtract = &RigVMFunction_MathDoubleSub(A=1, B=@n1.pose_alpha)
   /Script/ControlRig.RigUnit_HierarchySetPoseItemArray(ElementType=All, ItemsToSet=(), Pose=@n1.bind_pose, ExecutePin=@SequenceAggregate.A, Weight=@Subtract.Result) {is=struct RigVMUnitNode}
   &RigUnit_HierarchySetPoseItemArray(ElementType=All, ItemsToSet=(), Weight=@n1.pose_alpha, Pose=@n1.target_pose, ExecutePin=@SequenceAggregate.B)
   > RigVMFunctionReturnNode()
  ]}
]
hierarchy [
 root {loc=<0 0 0>, rot=<0 0 0 1>}
 pelvis {parent=root, loc=<0 -0.5301 98.0571>, rot=<0.0127 0.707 -0.0127 -0.707>}
 spine_01 {parent=pelvis, loc=<6.1208 -0.9328 0>, rot=<0 0 0.034 -0.9994>}
 spine_02 {parent=spine_01, loc=<6.5996 -0.2095 0>, rot=<0 0 0.0238 -0.9997>}
 spine_03 {parent=spine_02, loc=<6.4914 -0.1237 0>, rot=<0 0 0.0149 -0.9999>}
 spine_04 {parent=spine_03, loc=<6.9598 0.0794 0>, rot=<0 0 -0.0132 -0.9999>}
 spine_05 {parent=spine_04, loc=<6.9987 0 0>, rot=<0 0 -0.0015 -1>}
 neck_01 {parent=spine_05, loc=<11.3138 0.1529 0>, rot=<0 0 -0.2003 -0.9797>}
 neck_02 {parent=neck_01, loc=<4.5627 0 0>, rot=<0 0 0.0504 -0.9987>}
 head {parent=neck_02, loc=<5.3914 0 0>, rot=<0 0 0.115 -0.9934>}
 hair_rear_01_root {parent=head, loc=<6.5084 -8.439 1.3598>, rot=<0.1304 -0.9915 0 0>}
 hair_rear_01 {parent=hair_rear_01_root, loc=<0 0 0>, rot=<0 0 0 1>}
 hair_rear_02 {parent=hair_rear_01, loc=<5.7971 0 0>, rot=<0 0 0 1>}
 hair_rear_03 {parent=hair_rear_02, loc=<5.6639 0 0>, rot=<0 0 0.0037 -1>}
 hair_rear_04 {parent=hair_rear_03, loc=<6.9167 0 0>, rot=<0 0 -0.0952 -0.9955>}
 hair_rear_05 {parent=hair_rear_04, loc=<8.9086 0 0>, rot=<0 0 -0.0391 -0.9992>}
 hair_rear_06 {parent=hair_rear_05, loc=<6.695 0 0>, rot=<0 0 0 1>}
 hair_r_rear_01_root {parent=head, loc=<6.1662 -6.0998 6.7786>, rot=<0.1155 -0.959 0.2573 0.0288>}
 hair_r_rear_01 {parent=hair_r_rear_01_root, loc=<0 0 0>, rot=<0 0 0 1>}
 hair_r_rear_02 {parent=hair_r_rear_01, loc=<5.7971 0 0>, rot=<0 0 0 1>}
 hair_r_rear_03 {parent=hair_r_rear_02, loc=<5.6639 0 0>, rot=<-0.0275 -0.0089 0.0804 -0.9963>}
 hair_r_rear_04 {parent=hair_r_rear_03, loc=<6.9167 0 0>, rot=<0.2045 0.1762 -0.0829 -0.9593>}
 hair_r_rear_05 {parent=hair_r_rear_04, loc=<9.3055 0 0>, rot=<0.679 0.0423 -0.1428 -0.7188>}
 hair_r_rear_07 {parent=hair_r_rear_05, loc=<6.695 0 0>, rot=<-0.6895 0 0 -0.7242>}
 hair_r_rear_08 {parent=hair_r_rear_04, loc=<3.1505 0.9038 -2.4654>, rot=<-0.0379 -0.2421 0.0504 -0.9682>}
 hair_r_rear_09 {parent=hair_r_rear_08, loc=<5.4206 0 0>, rot=<0.023 0.2445 -0.1574 -0.9565>}
 hair_r_rear_010 {parent=hair_r_rear_09, loc=<5.0653 0 0>, rot=<0 -0.2344 0 -0.9721>}
 hair_l_rear_01_root {parent=head, loc=<6.0554 -7.7955 -2.6943>, rot=<0.1327 -0.979 -0.1536 -0.0195>}
 hair_l_rear_01 {parent=hair_l_rear_01_root, loc=<0 0 0>, rot=<0 0 0 1>}
 hair_l_rear_02 {parent=hair_l_rear_01, loc=<5.8203 0 0>, rot=<0.1899 0 0 -0.9818>}
 hair_l_rear_03 {parent=hair_l_rear_02, loc=<5.2344 0 0>, rot=<-0.1379 0.0195 -0.0396 -0.9895>}
 hair_l_rear_04 {parent=hair_l_rear_03, loc=<6.8798 0 0>, rot=<-0.2504 -0.0493 0.0179 -0.9667>}
 hair_l_rear_05 {parent=hair_l_rear_04, loc=<8.3066 0 0>, rot=<0.2344 0.0315 -0.0768 -0.9686>}
 hair_l_rear_06 {parent=hair_l_rear_05, loc=<6.695 0 0>, rot=<0.5969 0 0 -0.8023>}
 hair_r_side_01_root {parent=head, loc=<4.9273 -2.7871 8.0206>, rot=<0.1809 -0.981 -0.0635 0.0291>}
 hair_r_side_01 {parent=hair_r_side_01_root, loc=<0 0 0>, rot=<0 0 0 1>}
 hair_r_side_02 {parent=hair_r_side_01, loc=<5.7971 0 0>, rot=<0 -0.0367 0.0183 -0.9992>}
 hair_r_side_03 {parent=hair_r_side_02, loc=<5.6639 0 0>, rot=<0 -0.0033 0.0016 -1>}
 hair_r_side_04 {parent=hair_r_side_03, loc=<6.7153 -0.9645 -0.168>, rot=<-0.095 0.1448 -0.0519 -0.9835>}
 hair_r_side_05 {parent=hair_r_side_04, loc=<9.3055 0 0>, rot=<-0.0161 -0.0591 -0.1367 -0.9887>}
 hair_r_side_06 {parent=hair_r_side_05, loc=<6.695 0 0>, rot=<-0.0379 0 0 -0.9993>}
 hair_r_side_07 {parent=hair_r_side_04, loc=<3.1505 1.0911 -2.3884>, rot=<-0.0752 -0.2439 0.041 -0.966>}
 hair_r_side_08 {parent=hair_r_side_07, loc=<5.4206 0 0>, rot=<0.023 0.2445 -0.1574 -0.9565>}
 hair_r_side_09 {parent=hair_r_side_08, loc=<5.0653 0 0>, rot=<0 -0.2344 0 -0.9721>}
 hair_r_side_010 {parent=hair_r_side_03, loc=<1.9393 3.3692 -1.8331>, rot=<-0.3309 -0.1944 0.2643 -0.8848>}
 hair_r_side_011 {parent=hair_r_side_010, loc=<4.8922 0 0>, rot=<0.2404 0.3032 -0.0991 -0.9168>}
 hair_r_side_012 {parent=hair_r_side_011, loc=<2.5066 0 0>, rot=<-0.0514 0.021 -0.2748 -0.9599>}
 hair_r_front_01_root {parent=head, loc=<9.5126 7.707 3.2755>, rot=<0.356 -0.5345 0.7542 0.137>}
 hair_r_front_01 {parent=hair_r_front_01_root, loc=<0 0 0>, rot=<0 0 0 1>}
 hair_r_front_02 {parent=hair_r_front_01, loc=<6.0525 2.0019 1.5934>, rot=<-0.0263 -0.0873 -0.2686 -0.9589>}
 hair_r_front_03 {parent=hair_r_front_02, loc=<5.7971 0 0>, rot=<-0.0034 0.0002 -0.0565 -0.9984>}
 hair_r_front_04 {parent=hair_r_front_03, loc=<6.6746 0 0>, rot=<-0.2494 -0.1811 0.2372 -0.9213>}
 hair_r_front_05 {parent=hair_r_front_04, loc=<6.9443 0 0>, rot=<-0.187 -0.0927 0.0544 -0.9765>}
 hair_r_front_06 {parent=hair_r_front_05, loc=<6.3819 0 0>, rot=<0.0205 0.2264 -0.0496 -0.9726>}
 hair_r_front_07 {parent=hair_r_front_06, loc=<4.4469 0 0>, rot=<-0.0664 -0.2432 -0.0532 -0.9662>}
 hair_r_front_08 {parent=hair_r_front_04, loc=<2.7796 1.0196 3.3708>, rot=<-0.0202 -0.0091 0.2078 -0.9779>}
 hair_r_front_09 {parent=hair_r_front_08, loc=<4.9599 0 0>, rot=<0 0.1647 0 -0.9863>}
 hair_l_side_01_root {parent=head, loc=<5.6778 -5.3354 -5.2861>, rot=<0.1178 -0.9902 0.0749 0.0049>}
 hair_l_side_01 {parent=hair_l_side_01_root, loc=<0 0 0>, rot=<0 0 0 1>}
 hair_l_side_02 {parent=hair_l_side_01, loc=<5.7971 0 0>, rot=<-0.0219 0.0227 0.0699 -0.9971>}
 hair_l_side_03 {parent=hair_l_side_02, loc=<5.6639 0 0>, rot=<-0.01 0.0031 0.002 -0.9999>}
 hair_l_side_04 {parent=hair_l_side_03, loc=<6.8152 -0.6425 -0.9204>, rot=<0.0152 -0.0258 -0.0792 -0.9964>}
 hair_l_side_05 {parent=hair_l_side_04, loc=<8.3066 0 0>, rot=<-0.0428 0.002 -0.083 -0.9956>}
 hair_l_side_06 {parent=hair_l_side_05, loc=<6.695 0 0>, rot=<0.0202 0 0 -0.9998>}
 hair_l_side_07 {parent=hair_l_side_03, loc=<5.4815 2.0784 1.6002>, rot=<-0.0804 0.1926 0.0899 -0.9738>}
 hair_l_side_08 {parent=hair_l_side_07, loc=<5.0353 0 0>, rot=<0.0534 -0.164 -0.0992 -0.98>}
 hair_l_side_09 {parent=hair_l_side_08, loc=<3.5786 0 0>, rot=<-0.0125 -0.0444 -0.266 -0.9629>}
 hair_l_braid_01_root {parent=head, loc=<5.0669 0.8482 -6.9328>, rot=<-0.1701 0.6677 0.7223 -0.0591>}
 hair_l_braid_01 {parent=hair_l_braid_01_root, loc=<0 0 0>, rot=<0 0 0 1>}
 hair_l_braid_02 {parent=hair_l_braid_01, loc=<5.7971 0 0>, rot=<0.0536 -0.0506 -0.011 -0.9972>}
 hair_l_braid_03 {parent=hair_l_braid_02, loc=<4.0179 -0.0987 -0.1629>, rot=<-0.1669 -0.5083 -0.1084 -0.8378>}
 hair_l_braid_04 {parent=hair_l_braid_03, loc=<3.6938 0 0>, rot=<0 0 0 1>}
 hair_r_front_01_ctrl_null {parent=head, type=Null, loc=<9.5126 7.707 3.2755>, rot=<-0.356 0.5345 -0.7542 -0.137>}
 hair_r_front_01_ctrl {controlType=EULER_TRANSFORM, parent=hair_r_front_01_ctrl_null, type=Control, loc=<0 0 0>, rot=<0 0 0 1>}
 hair_r_front_02_ctrl_null {parent=hair_r_front_01_ctrl, type=Null, loc=<2.2956 5.4551 3.5103>, rot=<-0.1284 0.6003 -0.7315 0.2967>}
 hair_r_front_02_ctrl {controlType=EULER_TRANSFORM, parent=hair_r_front_02_ctrl_null, type=Control, loc=<0 0 0>, rot=<0 0 0 1>}
 hair_r_front_03_ctrl_null {parent=hair_r_front_02_ctrl, type=Null, loc=<-1.0109 -1.735 -1.6061>, rot=<-0.028 -0.0868 -0.2137 0.9726>}
 hair_r_front_03_ctrl {controlType=EULER_TRANSFORM, parent=hair_r_front_03_ctrl_null, type=Control, loc=<0 0 0>, rot=<0 0 0 1>}
 hair_r_front_04_ctrl_null {parent=hair_r_front_03_ctrl, type=Null, loc=<0.8719 -0.0989 0>, rot=<0.256 0.1661 -0.2895 0.9072>}
 hair_r_front_04_ctrl {controlType=EULER_TRANSFORM, parent=hair_r_front_04_ctrl_null, type=Control, loc=<0 0 0>, rot=<0 0 0 1>}
 hair_r_front_05_ctrl_null {parent=hair_r_front_04_ctrl, type=Null, loc=<0.2216 0.1422 0.0581>, rot=<-0.0833 -0.0607 0.1922 0.9759>}
 hair_r_front_05_ctrl {controlType=EULER_TRANSFORM, parent=hair_r_front_05_ctrl_null, type=Control, loc=<0 0 0>, rot=<0 0 0 1>}
 hair_r_front_06_ctrl_null {parent=hair_r_front_05_ctrl, type=Null, loc=<-0.5494 -0.0793 -0.0903>, rot=<-0.4228 -0.3292 0.0599 0.8422>}
 hair_r_front_06_ctrl {controlType=EULER_TRANSFORM, parent=hair_r_front_06_ctrl_null, type=Control, loc=<0 0 0>, rot=<0 0 0 1>}
 hair_r_front_07_ctrl_null {parent=hair_r_front_06_ctrl, type=Null, loc=<-1.7271 -0.275 0.8281>, rot=<-0.4767 0.0435 0.4032 0.7799>}
 hair_r_front_07_ctrl {controlType=EULER_TRANSFORM, parent=hair_r_front_07_ctrl_null, type=Control, loc=<0 0 0>, rot=<0 0 0 1>}
 clavicle_l {parent=spine_05, loc=<5.4831 -0.1053 -3.2946>, rot=<0.0465 -0.7289 -0.009 -0.6829>}
 upperarm_l {parent=clavicle_l, loc=<9.8658 -1.2136 -0.4105>, rot=<0.0164 -0.3245 -0.0269 -0.9454>}
 upperarm_twist_01_l {parent=upperarm_l, loc=<7.2135 -0.1134 0>, rot=<0 0 0 1>}
 clavicle_r {parent=spine_05, loc=<5.4831 -0.1053 3.2946>, rot=<0.7289 0.0465 -0.6829 0.009>}
 upperarm_r {parent=clavicle_r, loc=<-9.8658 1.2136 0.4105>, rot=<0.0164 -0.3245 -0.0269 -0.9454>}
 upperarm_twist_01_r {parent=upperarm_r, loc=<-7.2135 0.1134 0>, rot=<0 0 0 -1>}
 deform_upperarm_bk_base_r {parent=upperarm_twist_01_r, loc=<6.4053 4.3746 -0.314>, rot=<0 0 0 -1>}
 deform_upperarm_bk_r {parent=deform_upperarm_bk_base_r, loc=<0 0 0>, rot=<0 0 0 1>}
 deform_upperarm_fr_base_r {parent=upperarm_twist_01_r, loc=<5.6312 -5.092 -0.3354>, rot=<0 0 0 -1>}
 deform_upperarm_fr_r {parent=deform_upperarm_fr_base_r, loc=<0 0 0>, rot=<0 0 0 1>}
 deform_upperarm_up_base_r {parent=upperarm_twist_01_r, loc=<8.891 -0.4825 -3.9517>, rot=<0 0 0 -1>}
 deform_upperarm_up_r {parent=deform_upperarm_up_base_r, loc=<0 0 0>, rot=<0 0 0 1>}
 attach_cape {parent=spine_05, loc=<9.3216 -8.4501 0>, rot=<0.0616 -0.7044 -0.0616 -0.7044>}
 attach_backpack {parent=spine_05, loc=<0 -8.35 0>, rot=<-0.0093 -0.707 0.0093 -0.707>}
 deform_pec_aim_base_r {parent=spine_04, loc=<3.2452 9.3309 1.2237>, rot=<-0.4033 -0.0603 0.8507 0.3318>}
 deform_pec_aim_r {parent=deform_pec_aim_base_r, loc=<0 0 0>, rot=<0 0 0 1>}
 deform_pec_base_r {parent=deform_pec_aim_r, loc=<-9.3667 -0.0003 -0.0004>, rot=<0 0 0 1>}
 deform_pec_r {parent=deform_pec_base_r, loc=<0 0 0>, rot=<0 0 0 1>}
 hair_rear_06_ctrl_2 {controlType=EULER_TRANSFORM, parent=spine_04, type=Control, loc=<0 0 0>, rot=<0 0 0 1>}
 hair_l_side_06_ctrl_2 {controlType=EULER_TRANSFORM, parent=spine_04, type=Control, loc=<0 0 0>, rot=<0 0 0 1>}
 hair_r_side_06_ctrl_2 {controlType=EULER_TRANSFORM, parent=spine_04, type=Control, loc=<0 0 0>, rot=<0 0 0 1>}
 dyn_jacket_r_flap_upper_01_root {parent=spine_04, loc=<9.699 9.2868 4.3551>, rot=<-0.0877 -0.7429 0.1845 -0.6374>}
 dyn_jacket_r_flap_upper_01 {parent=dyn_jacket_r_flap_upper_01_root, loc=<0 0 0>, rot=<0 0 0 1>}
 dyn_jacket_r_flap_upper_02 {parent=dyn_jacket_r_flap_upper_01, loc=<-4.1684 0.0002 0.0005>, rot=<0 0 0 1>}
 dyn_jacket_r_rear_root {parent=spine_03, loc=<0.1533 -8.6612 9.5425>, rot=<0.9578 0.0272 0.0304 0.2846>}
 dyn_jacket_r_rear {parent=dyn_jacket_r_rear_root, loc=<0 0 0>, rot=<0 0 0 1>}
 dyn_jacket_r_front_root {parent=spine_03, loc=<-4.3815 9.369 10.0711>, rot=<-0.2872 -0.0891 -0.0732 -0.9509>}
 dyn_jacket_rear_root {parent=spine_03, loc=<1.1725 -10.525 0>, rot=<0.0836 -0.9965 0 0>}
 dyn_jacket_rear {parent=dyn_jacket_rear_root, loc=<0 0 0>, rot=<0 0 0 1>}
 head_ctrl {controlType=EULER_TRANSFORM, type=Control, loc=<0 0 0>, rot=<0 0 0 1>}
]
variables [
 HierarchyGetPoseItemArray_1_1_2_1_Pose {type=FRigPose, default=(Elements=,HierarchyTopologyVersion=-1,PoseHash=-1)}
 HierarchyGetPoseItemArray_1_1_3_1_1_Pose {type=FRigPose, default=(Elements=,HierarchyTopologyVersion=-1,PoseHash=-1)}
 HierarchyGetPoseItemArray_5_1_1_1_Pose {type=FRigPose, default=(Elements=,HierarchyTopologyVersion=-1,PoseHash=-1)}
 HierarchyGetPoseItemArray_5_1_1_2_1_1_Pose {type=FRigPose, default=(Elements=,HierarchyTopologyVersion=-1,PoseHash=-1)}
 HierarchyGetPoseItemArray_5_1_1_2_1_Pose {type=FRigPose, default=(Elements=,HierarchyTopologyVersion=-1,PoseHash=-1)}
 IsJacketVariant {type=bool, default=False}
 SphericalPoseReader_1_3_1_OutputParam {type=float, default=0}
 clavicle_r_raised_r_side_ctrls_pose {type=FRigPose, default=(Elements=,HierarchyTopologyVersion=-1,PoseHash=-1)}
 front_hair_arm_Raise_jacket_upper_pose {type=FRigPose, default=(Elements=,HierarchyTopologyVersion=-1,PoseHash=-1)}
 front_hair_arm_raise_jacket_mid_pose {type=FRigPose, default=(Elements=,HierarchyTopologyVersion=-1,PoseHash=-1)}
 front_hair_arm_raise_jacket_pose {type=FRigPose, default=(Elements=,HierarchyTopologyVersion=-1,PoseHash=-1)}
 front_hair_arm_raise_no_jacket_pose {type=FRigPose, default=(Elements=,HierarchyTopologyVersion=-1,PoseHash=-1)}
 front_hair_bones {type="TArray<FRigElementKey>"}
 front_hair_clavicle_raised_and_fwd_pose {type=FRigPose, default=(Elements=,HierarchyTopologyVersion=-1,PoseHash=-1)}
 front_hair_ctrl_bind {type=FRigPose, default=(Elements=,HierarchyTopologyVersion=-1,PoseHash=-1)}
 front_r_hair_ctrls {type="TArray<FRigElementKey>"}
 hair_aim_ctrls_bind {type=FRigPose, default=(Elements=,HierarchyTopologyVersion=-1,PoseHash=-1)}
 hair_aim_ctrls_head_back_pose {type=FRigPose, default=(Elements=,HierarchyTopologyVersion=-1,PoseHash=-1)}
 head_back_l_side_ctrls_pose {type=FRigPose, default=(Elements=,HierarchyTopologyVersion=-1,PoseHash=-1)}
 head_back_r_side_ctrls_pose {type=FRigPose, default=(Elements=,HierarchyTopologyVersion=-1,PoseHash=-1)}
 head_back_rear_ctrls_pose {type=FRigPose, default=(Elements=,HierarchyTopologyVersion=-1,PoseHash=-1)}
 r_arm_fwd_lower_ctrls_pose {type=FRigPose, default=(Elements=,HierarchyTopologyVersion=-1,PoseHash=-1)}
 r_clav_fwd_upper_ctrls_pose {type=FRigPose, default=(Elements=,HierarchyTopologyVersion=-1,PoseHash=-1)}
 rear_hair_bones {type="TArray<FRigElementKey>"}
 rear_hair_ctrl_bind {type=FRigPose, default=(Elements=,HierarchyTopologyVersion=-1,PoseHash=-1)}
 rear_hair_ctrls {type="TArray<FRigElementKey>"}
 rear_l_hair_bone {type="TArray<FRigElementKey>"}
 rear_l_hair_ctrl_bind {type=FRigPose, default=(Elements=,HierarchyTopologyVersion=-1,PoseHash=-1)}
 rear_l_hair_ctrls {type="TArray<FRigElementKey>"}
 rear_r_hair_bones {type="TArray<FRigElementKey>"}
 rear_r_hair_ctrl_bind {type=FRigPose, default=(Elements=,HierarchyTopologyVersion=-1,PoseHash=-1)}
 rear_r_hair_ctrls {type="TArray<FRigElementKey>"}
 side_l_hair_bones {type="TArray<FRigElementKey>"}
 side_l_hair_ctrl_bind {type=FRigPose, default=(Elements=,HierarchyTopologyVersion=-1,PoseHash=-1)}
 side_l_hair_ctrls {type="TArray<FRigElementKey>"}
 side_r_hair_bones {type="TArray<FRigElementKey>"}
 side_r_hair_ctrl_bind {type=FRigPose, default=(Elements=,HierarchyTopologyVersion=-1,PoseHash=-1)}
 side_r_hair_ctrls {type="TArray<FRigElementKey>"}
 spline_samples {type=int32, default=8}
 spr_arm_raise_jacket_mid {type=float, default=0}
 spr_arm_raise_jacket_upper {type=float, default=0}
 spr_arm_raise_lower_right {type=float, default=0}
 spr_arm_raise_upper_right {type=float, default=0}
 spr_clavicle_raise_left {type=float, default=0}
 spr_clavicle_raise_right {type=float, default=0}
 spr_clavicle_up_and_forward {type=float, default=0}
 spr_head_back {type=float, default=0}
 spr_head_left {type=float, default=0}
 spr_head_right {type=float, default=0}
]
