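# Physics asset for the FaceAcc front part of the medium female RollerBlade body, covering the right-front hair chain.
# Box bodies on the upper hair segments lead to capsules, with the chain forking at segment 04 into two branches.
# Hair-segment constraints use narrow per-joint swing and twist limits.
# Head, neck, upper spine, right clavicle and upper arm supply body shapes, joined by free-rotation constraints, as is the hair root to head.
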PhysicsAsset /BRCosmetics/Characters/Player/Female/Medium/Bodies/F_MED_RollerBlade/Meshes/Parts/F_MED_RollerBlade_FaceAcc_Front_FrontEnd_Physics
bodies {
 clavicle_r [{
   shapeType=Capsule
   length=11.2188
   radius=6.2863
   shapeName=clavicle_r_capsule
   center=<-7.2728 1.1238 1.1822>
   rotation=<rot 78.321 -0.1789 -2.2919>
  }]
 hair_r_front_01 [{shapeType=Box, extentX=7.5, extentY=2.375, extentZ=7.5, shapeName=hair_r_front_01_box, center=<2.9194 0 1.2934>}]
 hair_r_front_01_root [{shapeType=Capsule, length=0.505, radius=0.505, shapeName=hair_r_front_01_root_capsule, rotation=<rot 0 -90 0>}]
 hair_r_front_02 [{
   shapeType=Box
   extentX=7.5
   extentY=2.375
   extentZ=7.5
   shapeName=hair_r_front_02_box
   center=<0.5905 -1.3408 0>
   rotation=<rot -0.0002 -9.2369 0.0001>
  }]
 hair_r_front_03 [{
   shapeType=Box
   extentX=13.3125
   extentY=2.75
   extentZ=6.25
   shapeName=hair_r_front_03_box
   center=<5.0116 -0.9618 -0.4105>
   rotation=<rot -27.6652 -0.0001 0>
  }]
 hair_r_front_04 [{
   shapeType=Capsule
   length=0.6875
   radius=2.422
   shapeName=hair_r_front_04_capsule
   center=<1.2386 -0.9285 -0.2149>
   rotation=<rot 80.4223 -2.223 -7.4125>
  }]
 hair_r_front_05 [{
   shapeType=Capsule
   length=5.8868
   radius=2.7731
   shapeName=hair_r_front_05_capsule
   center=<2.463 0.1677 -0.2198>
   rotation=<rot 86.5684 179.9982 -175.4987>
  }]
 hair_r_front_06 [{
   shapeType=Capsule
   length=3.5706
   radius=1.6123
   shapeName=hair_r_front_06_capsule
   center=<1.0768 -0.0391 0.1941>
   rotation=<rot 74.8804 0.0002 4.049>
  }]
 hair_r_front_07 [{shapeType=Capsule, length=4.6936, radius=1.5416, shapeName=hair_r_front_07_capsule, center=<2.3692 0.3649 0>, rotation=<rot 90 0 -16.31>}]
 hair_r_front_08 [{shapeType=Capsule, length=3.0976, radius=1.6151, shapeName=hair_r_front_08_capsule, center=<1.688 -0.3413 0.9157>, rotation=<rot 90 0 0>}]
 hair_r_front_09 [{
   shapeType=Capsule
   length=2.3499
   radius=1.5367
   shapeName=hair_r_front_09_capsule
   center=<1.2492 0.2744 0>
   rotation=<rot 90 59.9892 48.9919>
  }]
 head [{shapeType=Capsule, length=7.7033, radius=5.4414, center=<0.9666 4.1136 0>, rotation=<rot 90 0 -12.383>}]
 neck_01 [{shapeType=Capsule, length=13.255, radius=5.13, shapeName=neck_01_capsule, center=<0 0.4367 0>, rotation=<rot 0 -90.0011 -89.998>}]
 spine_04 [{shapeType=Capsule, length=13.625, radius=11.1875, shapeName=spine_04_capsule, center=<-0.1083 1.3847 0>, rotation=<rot 0 -90 0>}]
 spine_05 [
  {
   shapeType=Capsule
   length=26.4062
   radius=12.75
   shapeName=spine_05_capsule
   center=<15.7466 2.5633 23.1521>
   rotation=<rot -0.9731 6.5767 75.3006>
  }
  {
   shapeType=Box
   extentX=10.9375
   extentY=7.125
   extentZ=30.1875
   shapeName=spine_05_box
   center=<3.7551 1.5232 0>
   rotation=<rot 0 -33.6336 0>
  }
 ]
 upperarm_r [{
   shapeType=Capsule
   length=25.6925
   radius=5.1925
   shapeName=upperarm_r_capsule
   center=<-14.5556 -0.0021 -0.2153>
   rotation=<rot -0.0626 -89.9441 92.4255>
  }]
}
constraints [
 {
  bone1Name=upperarm_r
  bone2Name=clavicle_r
  swing1LimitDegrees=45
  swing1Motion=Free
  swing2LimitDegrees=45
  swing2Motion=Free
  twistLimitDegrees=45
  twistMotion=Free
 }
 {
  bone1Name=hair_r_front_01_root
  bone2Name=head
  swing1LimitDegrees=45
  swing1Motion=Free
  swing2LimitDegrees=45
  swing2Motion=Free
  twistLimitDegrees=45
  twistMotion=Free
 }
 {bone1Name=hair_r_front_01, bone2Name=hair_r_front_01_root, swing1LimitDegrees=3, swing2LimitDegrees=3, twistLimitDegrees=3}
 {bone1Name=hair_r_front_02, bone2Name=hair_r_front_01, swing1LimitDegrees=16, swing2LimitDegrees=6, twistLimitDegrees=3}
 {bone1Name=hair_r_front_03, bone2Name=hair_r_front_02, swing1LimitDegrees=32, swing2LimitDegrees=6, twistLimitDegrees=3}
 {bone1Name=hair_r_front_04, bone2Name=hair_r_front_03, swing1LimitDegrees=12, swing2LimitDegrees=12, twistLimitDegrees=4}
 {bone1Name=hair_r_front_05, bone2Name=hair_r_front_04, swing1LimitDegrees=18, swing2LimitDegrees=18, twistLimitDegrees=4}
 {bone1Name=hair_r_front_06, bone2Name=hair_r_front_05, swing1LimitDegrees=32, swing2LimitDegrees=32, twistLimitDegrees=4}
 {bone1Name=hair_r_front_07, bone2Name=hair_r_front_06, swing1LimitDegrees=32, swing2LimitDegrees=32, twistLimitDegrees=4}
 {bone1Name=hair_r_front_08, bone2Name=hair_r_front_04, swing1LimitDegrees=32, swing2LimitDegrees=32, twistLimitDegrees=4}
 {bone1Name=hair_r_front_09, bone2Name=hair_r_front_08, swing1LimitDegrees=10, swing2LimitDegrees=10, twistLimitDegrees=10}
 {
  bone1Name=head
  bone2Name=neck_01
  swing1LimitDegrees=45
  swing1Motion=Free
  swing2LimitDegrees=45
  swing2Motion=Free
  twistLimitDegrees=45
  twistMotion=Free
 }
 {
  bone1Name=spine_05
  bone2Name=spine_04
  swing1LimitDegrees=45
  swing1Motion=Free
  swing2LimitDegrees=45
  swing2Motion=Free
  twistLimitDegrees=45
  twistMotion=Free
 }
 {
  bone1Name=neck_01
  bone2Name=spine_05
  swing1LimitDegrees=45
  swing1Motion=Free
  swing2LimitDegrees=45
  swing2Motion=Free
  twistLimitDegrees=45
  twistMotion=Free
 }
 {
  bone1Name=clavicle_r
  bone2Name=spine_05
  swing1LimitDegrees=45
  swing1Motion=Free
  swing2LimitDegrees=45
  swing2Motion=Free
  twistLimitDegrees=45
  twistMotion=Free
 }
]
