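# Physics asset for the RollerBlade female medium character's front face-accessory part, the skydive variant, covering a right-side front hair chain.
# Boxes on upper hair segments and capsules on lower ones form a chain from the head that forks at segment 04.
# Swing and twist limits are tightest near the top and looser further down.
# Head, neck, upper spine and right shoulder/arm capsules supply body shapes, linked by free angular constraints.
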PhysicsAsset /BRCosmetics/Characters/Player/Female/Medium/Bodies/F_MED_RollerBlade/Meshes/Parts/F_MED_RollerBlade_FaceAcc_Front_Skydive_Physics
bodies {
 clavicle_r [
  {
   shapeType=Capsule
   length=11.5625
   radius=6.38
   shapeName=clavicle_r_capsule
   center=<-5.7914 1.4728 0.8135>
   rotation=<rot 76.5912 -0.0005 -0.0005>
  }
  {
   shapeType=Capsule
   length=17.0625
   radius=5.1875
   shapeName=clavicle_r_capsule
   center=<-13.8503 -3.8068 -0.8756>
   rotation=<rot 0 0 77.0467>
  }
 ]
 hair_r_front_01 [{shapeType=Box, extentX=7.5, extentY=2.375, extentZ=7.5, shapeName=hair_r_front_01_box, center=<2.9194 0 1.2934>}]
 hair_r_front_01_root [{shapeType=Capsule, length=0.505, radius=0.505, shapeName=hair_r_front_01_root_capsule, rotation=<rot 0 -90 0>}]
 hair_r_front_02 [{
   shapeType=Box
   extentX=7.5
   extentY=2.375
   extentZ=7.5
   shapeName=hair_r_front_02_box
   center=<0.5905 -1.3408 0>
   rotation=<rot -0.0002 -9.2369 0.0001>
  }]
 hair_r_front_03 [{shapeType=Box, extentX=7.5, extentY=4.47, extentZ=7.5, shapeName=hair_r_front_03_box, center=<0 -0.9618 0>, rotation=<rot -0.0001 2.6868 0>}]
 hair_r_front_04 [{
   shapeType=Capsule
   length=6.9358
   radius=1.8908
   shapeName=hair_r_front_04_capsule
   center=<2.9232 -0.0475 0>
   rotation=<rot 80.4225 -2.2233 0>
  }]
 hair_r_front_05 [{shapeType=Capsule, length=4.8556, radius=1.9606, shapeName=hair_r_front_05_capsule, center=<2.4333 0 0>, rotation=<rot 90 0 0>}]
 hair_r_front_06 [{shapeType=Capsule, length=3.1019, radius=1.8311, shapeName=hair_r_front_06_capsule, center=<1.9042 0 0>, rotation=<rot 90 0 0>}]
 hair_r_front_07 [{shapeType=Capsule, length=3.4748, radius=1.4478, shapeName=hair_r_front_07_capsule, center=<2.5054 0 0>, rotation=<rot 90 0 0>}]
 hair_r_front_08 [{shapeType=Capsule, length=4.4101, radius=2.4276, shapeName=hair_r_front_08_capsule, center=<1.9991 0 0>, rotation=<rot 90 0 0>}]
 hair_r_front_09 [{shapeType=Capsule, length=3.0687, radius=1.5992, shapeName=hair_r_front_09_capsule, center=<1.2492 0.2744 0>, rotation=<rot 90 0 0>}]
 head [{shapeType=Capsule, length=7.7033, radius=5.4414, center=<0.9666 4.1136 0>, rotation=<rot 90 0 -12.383>}]
 neck_01 [{shapeType=Capsule, length=13.255, radius=5.13, shapeName=neck_01_capsule, center=<0 0.4367 0>, rotation=<rot 0 -90.0011 -89.998>}]
 spine_04 [{shapeType=Capsule, length=13.25, radius=11.5625, shapeName=spine_04_capsule, center=<-0.8636 1.8644 0>, rotation=<rot 0 -90 0>}]
 spine_05 [
  {shapeType=Capsule, length=13.38, radius=8.1925, shapeName=spine_05_capsule, center=<2.067 0 0>, rotation=<rot 0 -90 0>}
  {shapeType=Capsule, length=46.4781, radius=15, shapeName=spine_05_capsule, center=<21.8441 -11.7167 0>, rotation=<rot 0 27.6727 0>}
 ]
 upperarm_r [{
   shapeType=Capsule
   length=25.6925
   radius=5.1925
   shapeName=upperarm_r_capsule
   center=<-14.5556 -0.0021 -0.2153>
   rotation=<rot -0.0626 -89.9441 92.4255>
  }]
}
constraints [
 {
  bone1Name=upperarm_r
  bone2Name=clavicle_r
  swing1LimitDegrees=45
  swing1Motion=Free
  swing2LimitDegrees=45
  swing2Motion=Free
  twistLimitDegrees=45
  twistMotion=Free
 }
 {
  bone1Name=hair_r_front_01_root
  bone2Name=head
  swing1LimitDegrees=45
  swing1Motion=Free
  swing2LimitDegrees=45
  swing2Motion=Free
  twistLimitDegrees=45
  twistMotion=Free
 }
 {bone1Name=hair_r_front_01, bone2Name=hair_r_front_01_root, swing1LimitDegrees=4, swing2LimitDegrees=4, twistLimitDegrees=3}
 {bone1Name=hair_r_front_02, bone2Name=hair_r_front_01, swing1LimitDegrees=4, swing2LimitDegrees=4, twistLimitDegrees=3}
 {bone1Name=hair_r_front_03, bone2Name=hair_r_front_02, swing1LimitDegrees=4, swing2LimitDegrees=4, twistLimitDegrees=3}
 {bone1Name=hair_r_front_04, bone2Name=hair_r_front_03, swing1LimitDegrees=12, swing2LimitDegrees=12, twistLimitDegrees=6}
 {bone1Name=hair_r_front_05, bone2Name=hair_r_front_04, swing1LimitDegrees=12, swing2LimitDegrees=12, twistLimitDegrees=6}
 {bone1Name=hair_r_front_06, bone2Name=hair_r_front_05, swing1LimitDegrees=12, swing2LimitDegrees=12, twistLimitDegrees=6}
 {bone1Name=hair_r_front_07, bone2Name=hair_r_front_06, swing1LimitDegrees=12, swing2LimitDegrees=12, twistLimitDegrees=6}
 {bone1Name=hair_r_front_08, bone2Name=hair_r_front_04, swing1LimitDegrees=12, swing2LimitDegrees=12, twistLimitDegrees=6}
 {bone1Name=hair_r_front_09, bone2Name=hair_r_front_08, swing1LimitDegrees=10, swing2LimitDegrees=10, twistLimitDegrees=10}
 {
  bone1Name=head
  bone2Name=neck_01
  swing1LimitDegrees=45
  swing1Motion=Free
  swing2LimitDegrees=45
  swing2Motion=Free
  twistLimitDegrees=45
  twistMotion=Free
 }
 {
  bone1Name=spine_05
  bone2Name=spine_04
  swing1LimitDegrees=45
  swing1Motion=Free
  swing2LimitDegrees=45
  swing2Motion=Free
  twistLimitDegrees=45
  twistMotion=Free
 }
 {
  bone1Name=neck_01
  bone2Name=spine_05
  swing1LimitDegrees=45
  swing1Motion=Free
  swing2LimitDegrees=45
  swing2Motion=Free
  twistLimitDegrees=45
  twistMotion=Free
 }
 {
  bone1Name=clavicle_r
  bone2Name=spine_05
  swing1LimitDegrees=45
  swing1Motion=Free
  swing2LimitDegrees=45
  swing2Motion=Free
  twistLimitDegrees=45
  twistMotion=Free
 }
]
