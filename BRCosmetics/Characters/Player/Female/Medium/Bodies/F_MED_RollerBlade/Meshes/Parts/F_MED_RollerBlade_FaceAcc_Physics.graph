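# Physics asset for the F_MED_RollerBlade female FaceAcc mesh, defining capsule bodies and constraints for hair physics.
# Left and right side chains with branches, a rear chain and a left braid hang from head-linked roots, alongside spine, neck, head, clavicle and upper-arm capsules.
# Angular limits generally widen toward tips, while the braid allows wide swing with minimal twist.
# Angularly free cross-constraints tie side chains to the rear chain and the braid tip to the left side.
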PhysicsAsset /BRCosmetics/Characters/Player/Female/Medium/Bodies/F_MED_RollerBlade/Meshes/Parts/F_MED_RollerBlade_FaceAcc_Physics
bodies {
 clavicle_l [{
   shapeType=Capsule
   length=9
   radius=5.8175
   shapeName=clavicle_r_capsule
   center=<8.1841 -1.6474 -1.3839>
   rotation=<rot 76.5912 -0.0028 172.8879>
  }]
 clavicle_r [{
   shapeType=Capsule
   length=9
   radius=5.8175
   shapeName=clavicle_r_capsule
   center=<-8.1841 1.6474 1.3839>
   rotation=<rot 76.5912 -0.0028 -7.1121>
  }]
 hair_l_braid_01 [{
   shapeType=Capsule
   length=3.733
   radius=1.4403
   shapeName=hair_l_braid_01_capsule
   center=<3.7686 -0.9415 -0.276>
   rotation=<rot -0.8623 -78.7775 -91.8203>
  }]
 hair_l_braid_01_root [{shapeType=Capsule, length=0.505, radius=0.505, shapeName=hair_l_braid_01_root_capsule, rotation=<rot 0 -90 0>}]
 hair_l_braid_02 [{
   shapeType=Capsule
   length=1.5013
   radius=1.3261
   shapeName=hair_l_braid_02_capsule
   center=<1.8878 0.0345 -0.014>
   rotation=<rot 0 87.397 -87.6599>
  }]
 hair_l_braid_03 [{
   shapeType=Capsule
   length=1.7198
   radius=1.0427
   shapeName=hair_l_braid_03_capsule
   center=<0.8997 -0.2454 0.2191>
   rotation=<rot 0 -97.5865 -80.2357>
  }]
 hair_l_braid_04 [{
   shapeType=Capsule
   length=0.505
   radius=0.505
   shapeName=hair_l_braid_04_capsule
   center=<-0.0507 -0.3078 -0.0205>
   rotation=<rot 0 -69.0696 -87.4238>
  }]
 hair_l_side_01 [{shapeType=Capsule, length=5.824, radius=2.4885, shapeName=hair_l_side_01_capsule, center=<0 0.1123 1.5652>, rotation=<rot 0 0 64.6215>}]
 hair_l_side_01_root [{shapeType=Capsule, length=0.505, radius=0.505, shapeName=hair_l_side_01_root_capsule, rotation=<rot 0 -90 0>}]
 hair_l_side_02 [{
   shapeType=Capsule
   length=4
   radius=3.125
   shapeName=hair_l_side_02_capsule
   center=<1.8375 -0.6687 0.2348>
   rotation=<rot -34.367 -92.9363 91.6591>
  }]
 hair_l_side_03 [{
   shapeType=Capsule
   length=4.25
   radius=2.375
   shapeName=hair_l_side_03_capsule
   center=<3.4657 -1.8512 0.2731>
   rotation=<rot -83.9372 171.0939 -177.9324>
  }]
 hair_l_side_04 [{
   shapeType=Capsule
   length=5.9375
   radius=3.0625
   shapeName=hair_l_side_04_capsule
   center=<4.2509 -0.6341 0.1261>
   rotation=<rot -86.6555 -179.9999 180>
  }]
 hair_l_side_05 [{
   shapeType=Capsule
   length=4.7673
   radius=2.6068
   shapeName=hair_l_side_05_capsule
   center=<1.7186 -0.6488 0.0203>
   rotation=<rot -3.1637 -83.2294 -63.6871>
  }]
 hair_l_side_06 [{
   shapeType=Capsule
   length=5.6862
   radius=1.4082
   shapeName=hair_l_side_06_capsule
   center=<2.8397 0.5788 0.0324>
   rotation=<rot 1.8096 100.2287 -108.1151>
  }]
 hair_l_side_07 [{
   shapeType=Capsule
   length=2.8486
   radius=1.0751
   shapeName=hair_l_side_07_capsule
   center=<2.9295 0.2592 0.6416>
   rotation=<rot 67.4504 32.0487 29.4983>
  }]
 hair_l_side_08 [{shapeType=Capsule, length=1.8719, radius=1.1913, shapeName=hair_l_side_08_capsule, center=<1.2253 0.1501 -0.2189>, rotation=<rot 90 0 0>}]
 hair_l_side_09 [{shapeType=Capsule, length=3.482, radius=1.2206, shapeName=hair_l_side_09_capsule, center=<3.25 -0.145 0.1224>, rotation=<rot 90 0 0>}]
 hair_r_side_01 [{
   shapeType=Capsule
   length=6.75
   radius=3.25
   shapeName=hair_r_side_01_capsule
   center=<0.7859 -1.0528 -1.2064>
   rotation=<rot 0 0.0001 -64.0919>
  }]
 hair_r_side_010 [{shapeType=Capsule, length=2.1021, radius=1.08, shapeName=hair_r_side_010_capsule, center=<1.8029 0.0876 -0.1703>, rotation=<rot 90 0 0>}]
 hair_r_side_011 [{shapeType=Capsule, length=2.1159, radius=1.2102, shapeName=hair_r_side_011_capsule, center=<0.7213 0.2971 0.3103>, rotation=<rot 90 0 0>}]
 hair_r_side_012 [{
   shapeType=Capsule
   length=2.4043
   radius=0.9966
   shapeName=hair_r_side_012_capsule
   center=<1.9448 -0.3124 -0.1486>
   rotation=<rot 90 -4.2321 7.9846>
  }]
 hair_r_side_01_root [{shapeType=Capsule, length=0.505, radius=0.505, shapeName=hair_r_side_01_root_capsule, rotation=<rot 0 -90 0>}]
 hair_r_side_02 [{
   shapeType=Capsule
   length=1.3125
   radius=3.25
   shapeName=hair_r_side_02_capsule
   center=<2.4024 0.8953 -1.225>
   rotation=<rot 9.9923 79.6928 -84.2628>
  }]
 hair_r_side_03 [{
   shapeType=Capsule
   length=3.0625
   radius=2.875
   shapeName=hair_l_side_03_capsule
   center=<4.4537 -0.5911 0.2066>
   rotation=<rot 2.7935 82.0184 -76.4137>
  }]
 hair_r_side_04 [{
   shapeType=Capsule
   length=3.5903
   radius=2.5079
   shapeName=hair_r_side_04_capsule
   center=<4.4003 -1.6897 0.8415>
   rotation=<rot -0.055 97.7498 -94.9568>
  }]
 hair_r_side_05 [{
   shapeType=Capsule
   length=6.1892
   radius=2.9581
   shapeName=hair_r_side_05_capsule
   center=<2.5502 -1.2795 0.0136>
   rotation=<rot 2.6985 90.8484 -95.393>
  }]
 hair_r_side_06 [{
   shapeType=Capsule
   length=4.2878
   radius=1.5699
   shapeName=hair_r_side_06_capsule
   center=<3.5027 0.9305 -0.7808>
   rotation=<rot 2.4779 -79.3996 -77.1432>
  }]
 hair_r_side_07 [{
   shapeType=Capsule
   length=3.2344
   radius=1.6251
   shapeName=hair_r_side_07_capsule
   center=<0.6481 -0.9613 -0.0723>
   rotation=<rot 86.2136 -179.0691 172.4228>
  }]
 hair_r_side_08 [{shapeType=Capsule, length=5.082, radius=1.962, shapeName=hair_r_side_08_capsule, center=<1.1452 -0.0592 0.2028>, rotation=<rot 90 0 0>}]
 hair_r_side_09 [{
   shapeType=Capsule
   length=4.5514
   radius=1.8015
   shapeName=hair_r_side_09_capsule
   center=<1.9765 0.7002 -0.3276>
   rotation=<rot 90 55.861 37.3943>
  }]
 hair_rear_01 [{shapeType=Capsule, length=6.1169, radius=2.8933, shapeName=hair_rear_01_capsule, center=<1.1627 -0.8379 0>}]
 hair_rear_01_root [{shapeType=Capsule, length=0.505, radius=0.505, shapeName=hair_rear_01_root_capsule, rotation=<rot 0 -90 0>}]
 hair_rear_02 [{
   shapeType=Capsule
   length=1.6419
   radius=4.1439
   shapeName=hair_rear_02_capsule
   center=<1.7858 -0.0218 0.0053>
   rotation=<rot 89.2822 -179.9901 180>
  }]
 hair_rear_03 [{
   shapeType=Capsule
   length=1.8951
   radius=3.443
   shapeName=hair_rear_03_capsule
   center=<0.8767 -0.5492 0.0033>
   rotation=<rot 89.75 -179.9997 180>
  }]
 hair_rear_04 [{
   shapeType=Capsule
   length=2.5
   radius=3.375
   shapeName=hair_rear_04_capsule
   center=<1.288 -0.8828 0.0039>
   rotation=<rot 89.6628 -179.9998 180>
  }]
 hair_rear_05 [{
   shapeType=Capsule
   length=4.1843
   radius=3.6179
   shapeName=hair_rear_05_capsule
   center=<1.6452 -1.3922 -0.2276>
   rotation=<rot 0.4221 -89.9997 -90.7538>
  }]
 hair_rear_06 [{
   shapeType=Capsule
   length=5.9864
   radius=2.612
   shapeName=hair_rear_06_capsule
   center=<4.1682 0.2408 -0.4242>
   rotation=<rot -4.6632 -83.6111 -91.5235>
  }]
 head [{shapeType=Capsule, length=6.1408, radius=7.0039, center=<1.8005 0.3151 0>, rotation=<rot 90 0 -12.383>}]
 neck_01 [{shapeType=Capsule, length=4.6829, radius=7.255, shapeName=neck_01_capsule, rotation=<rot 0 -90.0011 -89.998>}]
 spine_01 [{shapeType=Capsule, length=3.875, radius=12.88, shapeName=spine_01_capsule, center=<5.2041 2.8039 0>, rotation=<rot 0 -90 0>}]
 spine_04 [{shapeType=Capsule, length=13.25, radius=11.5625, shapeName=spine_04_capsule, center=<-0.8636 1.8644 0>, rotation=<rot 0 -90 0>}]
 spine_05 [{shapeType=Capsule, length=12.1925, radius=9.255, shapeName=spine_05_capsule, center=<-0.7459 -0.2816 0>, rotation=<rot 0 -90 0>}]
 upperarm_l [{
   shapeType=Capsule
   length=22.9425
   radius=5.505
   shapeName=upperarm_r_capsule
   center=<14.5556 0.0021 0.2153>
   rotation=<rot -0.0626 -89.9441 -87.5745>
  }]
 upperarm_r [{
   shapeType=Capsule
   length=21.13
   radius=7.63
   shapeName=upperarm_r_capsule
   center=<-14.5556 -0.0021 -0.2153>
   rotation=<rot -0.0626 -89.9441 92.4255>
  }]
}
constraints [
 {
  bone1Name=upperarm_r
  bone2Name=clavicle_r
  swing1LimitDegrees=45
  swing1Motion=Free
  swing2LimitDegrees=45
  swing2Motion=Free
  twistLimitDegrees=45
  twistMotion=Free
 }
 {
  bone1Name=head
  bone2Name=neck_01
  swing1LimitDegrees=45
  swing1Motion=Free
  swing2LimitDegrees=45
  swing2Motion=Free
  twistLimitDegrees=45
  twistMotion=Free
 }
 {
  bone1Name=spine_05
  bone2Name=spine_04
  swing1LimitDegrees=45
  swing1Motion=Free
  swing2LimitDegrees=45
  swing2Motion=Free
  twistLimitDegrees=45
  twistMotion=Free
 }
 {bone1Name=hair_rear_01, bone2Name=hair_rear_01_root, swing1LimitDegrees=4, swing2LimitDegrees=8, twistLimitDegrees=8}
 {bone1Name=hair_rear_02, bone2Name=hair_rear_01, swing1LimitDegrees=12, swing2LimitDegrees=12, twistLimitDegrees=12}
 {bone1Name=hair_rear_03, bone2Name=hair_rear_02, swing1LimitDegrees=18, swing2LimitDegrees=18, twistLimitDegrees=18}
 {bone1Name=hair_rear_04, bone2Name=hair_rear_03, swing1LimitDegrees=18, swing2LimitDegrees=18, twistLimitDegrees=18}
 {bone1Name=hair_rear_05, bone2Name=hair_rear_04, swing1LimitDegrees=32, swing2LimitDegrees=24, twistLimitDegrees=24}
 {
  bone1Name=upperarm_l
  bone2Name=clavicle_l
  swing1LimitDegrees=45
  swing1Motion=Free
  swing2LimitDegrees=45
  swing2Motion=Free
  twistLimitDegrees=45
  twistMotion=Free
 }
 {
  bone1Name=neck_01
  bone2Name=spine_05
  swing1LimitDegrees=45
  swing1Motion=Free
  swing2LimitDegrees=45
  swing2Motion=Free
  twistLimitDegrees=45
  twistMotion=Free
 }
 {
  bone1Name=hair_l_side_01_root
  bone2Name=head
  swing1LimitDegrees=45
  swing1Motion=Free
  swing2LimitDegrees=45
  swing2Motion=Free
  twistLimitDegrees=45
  twistMotion=Free
 }
 {bone1Name=hair_l_side_01, bone2Name=hair_l_side_01_root, swing1LimitDegrees=4, swing2LimitDegrees=4, twistLimitDegrees=2}
 {bone1Name=hair_l_side_02, bone2Name=hair_l_side_01, swing1LimitDegrees=12, swing2LimitDegrees=12, twistLimitDegrees=4}
 {bone1Name=hair_l_side_03, bone2Name=hair_l_side_02, swing1LimitDegrees=18, swing2LimitDegrees=18, twistLimitDegrees=1}
 {bone1Name=hair_l_side_04, bone2Name=hair_l_side_03, swing1LimitDegrees=18, swing2LimitDegrees=18, twistLimitDegrees=18}
 {bone1Name=hair_l_side_05, bone2Name=hair_l_side_04, swing1LimitDegrees=42, swing2LimitDegrees=24, twistLimitDegrees=2}
 {bone1Name=hair_l_side_07, bone2Name=hair_l_side_03, swing1LimitDegrees=4, swing2LimitDegrees=4, twistLimitDegrees=4}
 {bone1Name=hair_l_side_08, bone2Name=hair_l_side_07, swing1LimitDegrees=14, swing2LimitDegrees=14, twistLimitDegrees=8}
 {bone1Name=hair_l_side_09, bone2Name=hair_l_side_08, swing1LimitDegrees=20, swing2LimitDegrees=20, twistLimitDegrees=8}
 {
  bone1Name=hair_r_side_01_root
  bone2Name=head
  swing1LimitDegrees=45
  swing1Motion=Free
  swing2LimitDegrees=45
  swing2Motion=Free
  twistLimitDegrees=45
  twistMotion=Free
 }
 {bone1Name=hair_r_side_01, bone2Name=hair_r_side_01_root, swing1LimitDegrees=4, swing2LimitDegrees=4, twistLimitDegrees=2}
 {bone1Name=hair_r_side_02, bone2Name=hair_r_side_01, swing1LimitDegrees=12, swing2LimitDegrees=12, twistLimitDegrees=12}
 {bone1Name=hair_r_side_07, bone2Name=hair_r_side_04, swing1LimitDegrees=24, swing2LimitDegrees=24, twistLimitDegrees=24}
 {bone1Name=hair_r_side_08, bone2Name=hair_r_side_07, swing1LimitDegrees=24, swing2LimitDegrees=24, twistLimitDegrees=24}
 {bone1Name=hair_r_side_09, bone2Name=hair_r_side_08, swing1LimitDegrees=24, swing2LimitDegrees=24, twistLimitDegrees=24}
 {bone1Name=hair_r_side_010, bone2Name=hair_r_side_03, swing1LimitDegrees=8, swing2LimitDegrees=8, twistLimitDegrees=8}
 {bone1Name=hair_r_side_011, bone2Name=hair_r_side_010, swing1LimitDegrees=14, swing2LimitDegrees=14, twistLimitDegrees=8}
 {bone1Name=hair_r_side_012, bone2Name=hair_r_side_011, swing1LimitDegrees=20, swing2LimitDegrees=20, twistLimitDegrees=8}
 {
  bone1Name=hair_rear_01_root
  bone2Name=head
  swing1LimitDegrees=45
  swing1Motion=Free
  swing2LimitDegrees=45
  swing2Motion=Free
  twistLimitDegrees=45
  twistMotion=Free
 }
 {
  bone1Name=hair_l_braid_01_root
  bone2Name=head
  swing1LimitDegrees=45
  swing1Motion=Free
  swing2LimitDegrees=45
  swing2Motion=Free
  twistLimitDegrees=45
  twistMotion=Free
 }
 {bone1Name=hair_l_braid_01, bone2Name=hair_l_braid_01_root, swing1LimitDegrees=45, swing2LimitDegrees=45, twistLimitDegrees=1}
 {bone1Name=hair_l_braid_02, bone2Name=hair_l_braid_01, swing1LimitDegrees=45, swing2LimitDegrees=45, twistLimitDegrees=1}
 {bone1Name=hair_l_braid_03, bone2Name=hair_l_braid_02, swing1LimitDegrees=45, swing2LimitDegrees=45, twistLimitDegrees=1}
 {bone1Name=hair_l_braid_04, bone2Name=hair_l_braid_03, swing1LimitDegrees=45, swing2LimitDegrees=45, twistLimitDegrees=1}
 {
  bone1Name=hair_l_braid_04
  bone2Name=hair_l_side_03
  swing1LimitDegrees=45
  swing1Motion=Free
  swing2LimitDegrees=45
  swing2Motion=Free
  twistLimitDegrees=45
  twistMotion=Free
 }
 {bone1Name=hair_r_side_03, bone2Name=hair_r_side_02, swing1LimitDegrees=18, swing2LimitDegrees=18, twistLimitDegrees=18}
 {
  bone1Name=clavicle_r
  bone2Name=spine_05
  swing1LimitDegrees=45
  swing1Motion=Free
  swing2LimitDegrees=45
  swing2Motion=Free
  twistLimitDegrees=45
  twistMotion=Free
 }
 {bone1Name=hair_r_side_05, bone2Name=hair_r_side_04, swing1LimitDegrees=32, swing2LimitDegrees=28, twistLimitDegrees=28}
 {bone1Name=hair_r_side_04, bone2Name=hair_r_side_03, swing1LimitDegrees=18, swing2LimitDegrees=18, twistLimitDegrees=18}
 {
  bone1Name=hair_rear_05
  bone2Name=hair_r_side_05
  swing1LimitDegrees=45
  swing1Motion=Free
  swing2LimitDegrees=45
  swing2Motion=Free
  twistLimitDegrees=45
  twistMotion=Free
 }
 {
  bone1Name=hair_rear_05
  bone2Name=hair_l_side_05
  swing1LimitDegrees=45
  swing1Motion=Free
  swing2LimitDegrees=45
  swing2Motion=Free
  twistLimitDegrees=45
  twistMotion=Free
 }
 {bone1Name=hair_l_side_06, bone2Name=hair_l_side_05, swing1LimitDegrees=42, swing2LimitDegrees=24, twistLimitDegrees=8}
 {bone1Name=hair_r_side_06, bone2Name=hair_r_side_05, swing1LimitDegrees=32, swing2LimitDegrees=28, twistLimitDegrees=28}
 {
  bone1Name=hair_l_side_04
  bone2Name=hair_rear_04
  swing1LimitDegrees=45
  swing1Motion=Free
  swing2LimitDegrees=45
  swing2Motion=Free
  twistLimitDegrees=45
  twistMotion=Free
 }
 {
  bone1Name=hair_r_side_04
  bone2Name=hair_rear_04
  swing1LimitDegrees=45
  swing1Motion=Free
  swing2LimitDegrees=45
  swing2Motion=Free
  twistLimitDegrees=45
  twistMotion=Free
 }
 {bone1Name=hair_rear_06, bone2Name=hair_rear_05, swing1LimitDegrees=32, swing2LimitDegrees=28, twistLimitDegrees=28}
 {
  bone1Name=clavicle_l
  bone2Name=spine_05
  swing1LimitDegrees=45
  swing1Motion=Free
  swing2LimitDegrees=45
  swing2Motion=Free
  twistLimitDegrees=45
  twistMotion=Free
 }
 {
  bone1Name=spine_04
  bone2Name=spine_01
  swing1LimitDegrees=45
  swing1Motion=Free
  swing2LimitDegrees=45
  swing2Motion=Free
  twistLimitDegrees=45
  twistMotion=Free
 }
]
