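# Animation Blueprint for the RollerBlade female character's face accessory, adding secondary physics to front and rear hair.
# It copies the parent pose and the body's jacket-flap bones, then runs a jacket-aware Control Rig.
# Rear-hair RigidBody variants (jacket, backpack, DBNO) feed front-hair solvers: RigidBody or an AnimDynamics chain with planar limits against jacket and body.
# Events set jacket and back bling flags and a movement-state gravity override with skydive wind.
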AnimBlueprint /BRCosmetics/Characters/Player/Female/Medium/Bodies/F_MED_RollerBlade/Meshes/Parts/F_MED_RollerBlade_FaceAcc_AnimBP
target_skeleton=&/F_MED_RollerBlade_FaceAcc_Skeleton
graphs [
 AnimGraph {nodes [
   Output Pose(Result=Use cached pose 'PostFrontHairPhysics'()) {layerGroup=None}
   n1 = RigidBody!(
    ComponentPose=Local To Component(LocalPose=Use cached pose 'PostControlRig'())
    Alpha
    OverrideWorldGravity:Vector="0,0,-980"
    ExternalForce
    ComponentLinearAccScale:Vector="0.2,0.5,0.2"
    ComponentLinearVelScale:Vector="0.2,0.5,0.2"
   ) -> (Pose) {
    bClampLinearTranslationLimitToRefPose=true
    bOverrideWorldGravity=true
    overridePhysicsAsset=&/F_MED_RollerBlade_FaceAcc_Jacket_Physics
    simulationSpace=BaseBoneSpace
    alphaBoolBlend {blendOption=Linear}
    baseBoneRef {boneName=spine_05}
    componentAppliedLinearAccClamp=<1000 1000 1400>
    componentLinearAccScale=<0.3 0.3 0.5>
    componentLinearVelScale=<0.3 0.3 0.5>
    overrideWorldGravity {z=-800}
    simSpaceSettings {
     maxAngularAcceleration=200
     maxAngularVelocity=10
     maxLinearAcceleration=200
     maxLinearVelocity=200
     velocityScaleZ=0.2
     worldAlpha=0.25
     externalAngularVelocity=<0.2 0.2 0.2>
     externalLinearDragV=<0.2 0.2>
     externalLinearVelocity=<0.2 0.2 0.2>
    }
   }
   n2 = RigidBody(
    ComponentPose=Local To Component(LocalPose=Use cached pose 'PostControlRig'())
    OverrideWorldGravity="0,0,-980"
    ComponentLinearAccScale="0.2,0.5,0.2"
    ComponentLinearVelScale="0.2,0.5,0.2"
   ) {
    bClampLinearTranslationLimitToRefPose=true
    bOverrideWorldGravity=true
    overridePhysicsAsset=&/F_MED_RollerBlade_FaceAcc_Physics
    simulationSpace=BaseBoneSpace
    alphaBoolBlend {blendOption=Linear}
    baseBoneRef {boneName=spine_05}
    componentAppliedLinearAccClamp=<1000 1000 1400>
    componentLinearAccScale=<0.3 0.3 0.5>
    componentLinearVelScale=<0.3 0.3 0.5>
    overrideWorldGravity {z=-800}
    simSpaceSettings {
     maxAngularAcceleration=200
     maxAngularVelocity=10
     maxLinearAcceleration=200
     maxLinearVelocity=200
     velocityScaleZ=0.2
     worldAlpha=0.25
     externalAngularVelocity=<0.2 0.2 0.2>
     externalLinearDragV=<0.2 0.2>
     externalLinearVelocity=<0.2 0.2 0.2>
    }
   }
   n3 = Blend Poses by bool(BlendPose_0=Component To Local(ComponentPose=@n1), BlendPose_1=Component To Local(ComponentPose=@n2)) {blendType=Linear}
   n4 = RigidBody(
    ComponentPose=Local To Component(LocalPose=Use cached pose 'PostControlRig'())
    Alpha=0.8
    OverrideWorldGravity="0,0,-980"
    ComponentLinearAccScale="0.5,0.5,0.5"
    ComponentLinearVelScale="0.5,0.5,0.5"
   ) {
    bClampLinearTranslationLimitToRefPose=true
    bOverrideWorldGravity=true
    overridePhysicsAsset=&/F_MED_RollerBlade_FaceAcc_Backpack_Physics
    simulationSpace=BaseBoneSpace
    alphaBoolBlend {blendOption=Linear}
    baseBoneRef {boneName=spine_05}
    componentAppliedLinearAccClamp=<1000 1000 600>
    componentLinearAccScale=<0.3 0.3 0.5>
    componentLinearVelScale=<0.3 0.3 0.5>
    overrideWorldGravity {z=-800}
    simSpaceSettings {
     maxAngularAcceleration=200
     maxAngularVelocity=10
     maxLinearAcceleration=200
     maxLinearVelocity=200
     velocityScaleZ=0
     worldAlpha=0.25
     externalAngularVelocity=<0.2 0.2 0.2>
     externalLinearDragV=<0.2 0.2>
     externalLinearVelocity=<0.2 0.2 0.2>
    }
   }
   n5 = RigidBody(
    ComponentPose=Local To Component(LocalPose=Use cached pose 'PostControlRig'())
    OverrideWorldGravity="0,0,-980"
    ComponentLinearAccScale="0.2,0.2,0.2"
    ComponentLinearVelScale="0.2,0.2,0.2"
   ) {
    bClampLinearTranslationLimitToRefPose=true
    bOverrideWorldGravity=true
    overridePhysicsAsset=&F_MED_RollerBlade_FaceAcc_Backpack_Physics
    simulationSpace=BaseBoneSpace
    alphaBoolBlend {blendOption=Linear}
    baseBoneRef {boneName=spine_05}
    componentAppliedLinearAccClamp=<1000 1000 600>
    componentLinearAccScale=<0.3 0.3 0.5>
    componentLinearVelScale=<0.3 0.3 0.5>
    overrideWorldGravity {z=-800}
    simSpaceSettings {
     maxAngularAcceleration=200
     maxAngularVelocity=10
     maxLinearAcceleration=200
     maxLinearVelocity=200
     velocityScaleZ=0
     worldAlpha=0.25
     externalAngularVelocity=<0.2 0.2 0.2>
     externalLinearDragV=<0.2 0.2>
     externalLinearVelocity=<0.2 0.2 0.2>
    }
   }
   n6 = Blend Poses by bool(BlendPose_0=Component To Local(ComponentPose=@n4), BlendPose_1=Component To Local(ComponentPose=@n5)) {blendType=Linear}
   n7 = RigidBody(
    ComponentPose=Local To Component(LocalPose=Use cached pose 'PostControlRig'())
    OverrideWorldGravity="0,0,-980"
    ComponentLinearAccScale="0.2,0.2,0.2"
    ComponentLinearVelScale="0.2,0.2,0.2"
   ) {
    bClampLinearTranslationLimitToRefPose=true
    bOverrideWorldGravity=true
    overridePhysicsAsset=&/F_MED_RollerBlade_FaceAcc_DBNO_Physics
    simulationSpace=BaseBoneSpace
    alphaBoolBlend {blendOption=Linear}
    baseBoneRef {boneName=spine_05}
    componentAppliedLinearAccClamp=<1000 1000 600>
    componentLinearAccScale=<0.3 0.3 0.5>
    componentLinearVelScale=<0.3 0.3 0.5>
    overrideWorldGravity {z=-800}
    simSpaceSettings {
     maxAngularAcceleration=200
     maxAngularVelocity=10
     maxLinearAcceleration=200
     maxLinearVelocity=200
     velocityScaleZ=0
     worldAlpha=0.25
     externalAngularVelocity=<0.2 0.2 0.2>
     externalLinearDragV=<0.2 0.2>
     externalLinearVelocity=<0.2 0.2 0.2>
    }
   }
   Save cached pose 'PostRearHairPhysics'(Pose=Blend Poses by bool(BlendPose_0=Component To Local(ComponentPose=@n7), BlendPose_1=Blend Poses by bool(BlendPose_0=@n6, BlendPose_1=@n3) {blendType=Linear}) {blendType=Linear}) {cacheName=PostRearHairPhysics}
   Save cached pose 'PostFrontHairPhysics'(Pose=Blend Poses by bool(BlendPose_0=Use cached pose 'JacketFronHairAnimDynamics'(), BlendPose_1=Use cached pose 'FrontHairRigidbodyPhysics'()) {blendType=Linear}) {cacheName=PostFrontHairPhysics}
   n8 = Get BodyPart() -> (BodyPart:Skeletal Mesh Component Object Reference)
   n9 = Control Rig!(
    Source=Layered blend per bone(
     BasePose=Copy Pose From Mesh() {bUseAttachedParent=true}
     BlendPoses_0=Copy Pose From Mesh(SourceMeshComponent=@n8) {rootBoneToCopy=dyn_jacket_r_flap_upper_01_root}
    ) {layerSetup [{branchFilters [{boneName=dyn_jacket_r_flap_upper_01_root}]}]}
    Alpha
    IsJacketVariant:bool=Get WearingJacket() -> (WearingJacket:bool)
   ) -> (Pose) {
    alphaBoolBlend {blendOption=Linear}
    controlRigAssetReference {blueprintRigClass=&/F_MED_RollerBlade_FaceAcc_CtrlRig.F_MED_RollerBlade_FaceAcc_CtrlRig_C}
   }
   Save cached pose 'PostControlRig'(Pose=@n9) {cacheName=PostControlRig}
   n10 = Anim Dynamics - Bone: hair_r_front_03(ComponentPose=Local To Component(LocalPose=Use cached pose 'PostRearHairPhysics'()), Alpha=1, ExternalForce="0,0,0") {
    angularDampingOverride=1
    angularSpringConstant=50
    bAngularSpring=true
    bOverrideAngularDamping=true
    bUseGravityOverride=true
    gravityScale=0.35
    simulationSpace=RootRelative
    alphaBoolBlend {blendOption=Linear}
    boundBone {boneName=hair_r_front_03}
    componentAppliedLinearAccClamp=<800 800 800>
    componentLinearAccScale=<0.5 0.5 0.5>
    componentLinearVelScale=<0.5 0.5 0.5>
    gravityOverride {z=-3000}
    physicsBodyDefinitions [{
      boundBone {boneName=hair_r_front_03}
      boxExtents=<20 30 30>
      constraintSetup {angularLimitsMax=<2 2 0>, angularLimitsMin=<-2 -2 0>, angularTarget=<1 0 0>}
      localJointOffset=<4 0 0>
     }]
    planarLimits [{drivingBone {boneName=clavicle_r}, planeTransform {location=<0 -7 0>, rotation=<rot 0 0 -119.9999>}}]
   }
   n11 = Anim Dynamics - Bone: hair_r_front_04(ComponentPose=@n10, Alpha=1, ExternalForce="0,0,0") {
    angularDampingOverride=1
    angularSpringConstant=50
    bAngularSpring=true
    bOverrideAngularDamping=true
    bUseGravityOverride=true
    gravityScale=0.35
    simulationSpace=RootRelative
    alphaBoolBlend {blendOption=Linear}
    boundBone {boneName=hair_r_front_04}
    componentAppliedLinearAccClamp=<800 800 800>
    componentLinearAccScale=<0.5 0.5 0.5>
    componentLinearVelScale=<0.5 0.5 0.5>
    gravityOverride {z=-3000}
    physicsBodyDefinitions [{
      boundBone {boneName=hair_r_front_04}
      boxExtents=<20 30 30>
      constraintSetup {angularLimitsMax=<2 2 0>, angularLimitsMin=<-2 -2 0>, angularTarget=<1 0 0>}
      localJointOffset=<4 0 0>
     }]
    planarLimits [
     {drivingBone {boneName=dyn_jacket_r_flap_upper_01_root}, planeTransform {location=<0 -1 0>, rotation=<rot 0 0 -90>}}
     {drivingBone {boneName=clavicle_r}, planeTransform {location=<0 -8 0>, rotation=<rot 0 0 -119.9999>}}
     {drivingBone {boneName=spine_05}, planeTransform {location=<0 0 5>, rotation=<rot 0 0 0>}}
    ]
   }
   n12 = Anim Dynamics - Bone: hair_r_front_05(ComponentPose=@n11, Alpha=1, ExternalForce="0,0,0") {
    angularDampingOverride=1
    angularSpringConstant=50
    bAngularSpring=true
    bOverrideAngularDamping=true
    bUseGravityOverride=true
    gravityScale=0.35
    simulationSpace=RootRelative
    alphaBoolBlend {blendOption=Linear}
    boundBone {boneName=hair_r_front_05}
    componentAppliedLinearAccClamp=<800 800 800>
    componentLinearAccScale=<0.5 0.5 0.5>
    componentLinearVelScale=<0.5 0.5 0.5>
    gravityOverride {z=-3000}
    physicsBodyDefinitions [{
      boundBone {boneName=hair_r_front_05}
      boxExtents=<20 30 30>
      constraintSetup {angularLimitsMax=<10 10 10>, angularLimitsMin=<-10 -10 -10>, angularTarget=<1 0 0>}
      localJointOffset=<4 0 0>
     }]
    planarLimits [
     {drivingBone {boneName=clavicle_r}, planeTransform {location=<0 -8 0>, rotation=<rot 0 0 -119.9999>}}
     {drivingBone {boneName=dyn_jacket_r_flap_upper_01_root}, planeTransform {location=<0 -1 0>, rotation=<rot 0 0 -90>}}
    ]
   }
   n13 = Anim Dynamics - Bone: hair_r_front_06(ComponentPose=@n12, Alpha=1, ExternalForce="0,0,0") {
    angularDampingOverride=1
    angularSpringConstant=50
    bAngularSpring=true
    bOverrideAngularDamping=true
    bUseGravityOverride=true
    gravityScale=0.35
    simulationSpace=RootRelative
    alphaBoolBlend {blendOption=Linear}
    boundBone {boneName=hair_r_front_06}
    componentAppliedLinearAccClamp=<800 800 800>
    componentLinearAccScale=<0.5 0.5 0.5>
    componentLinearVelScale=<0.5 0.5 0.5>
    gravityOverride {z=-3000}
    physicsBodyDefinitions [{
      boundBone {boneName=hair_r_front_06}
      boxExtents=<20 30 30>
      constraintSetup {angularLimitsMax=<10 10 10>, angularLimitsMin=<-10 -10 -10>, angularTarget=<1 0 0>}
      localJointOffset=<4 0 0>
     }]
    planarLimits [
     {drivingBone {boneName=spine_05}, planeTransform {location=<0 13 0>, rotation=<rot 0 -10 90>, scale=<0.5 0.5 0.5>}}
     {drivingBone {boneName=dyn_jacket_r_flap_upper_01_root}, planeTransform {location=<0 -1 0>, rotation=<rot 0 10 -90>}}
     {drivingBone {boneName=spine_05}, planeTransform {location=<0 0 7>, rotation=<rot 0 90 0>}}
    ]
   }
   n14 = Anim Dynamics - Bone: hair_r_front_07(ComponentPose=@n13, Alpha=1, ExternalForce="0,0,0") {
    angularDampingOverride=1
    angularSpringConstant=50
    bAngularSpring=true
    bOverrideAngularDamping=true
    bUseGravityOverride=true
    gravityScale=0.35
    simulationSpace=RootRelative
    alphaBoolBlend {blendOption=Linear}
    boundBone {boneName=hair_r_front_07}
    componentAppliedLinearAccClamp=<800 800 800>
    componentLinearAccScale=<0.5 0.5 0.5>
    componentLinearVelScale=<0.5 0.5 0.5>
    gravityOverride {z=-3000}
    physicsBodyDefinitions [{
      boundBone {boneName=hair_r_front_07}
      boxExtents=<20 30 30>
      constraintSetup {angularLimitsMax=<10 10 10>, angularLimitsMin=<-10 -10 -10>, angularTarget=<1 0 0>}
      localJointOffset=<4 0 0>
     }]
    planarLimits [
     {drivingBone {boneName=clavicle_r}, planeTransform {location=<0 -7 0>, rotation=<rot 0 0 -119.9999>}}
     {drivingBone {boneName=spine_05}, planeTransform {location=<0 12 0>, rotation=<rot 0 -9.9999 90>, scale=<0.5 0.5 0.5>}}
     {drivingBone {boneName=dyn_jacket_r_flap_upper_01_root}, planeTransform {location=<-2 -1 0>, rotation=<rot 0 10 -90>}}
    ]
   }
   n15 = Anim Dynamics - Bone: hair_r_front_08(ComponentPose=@n14, Alpha=1, ExternalForce="0,0,0") {
    angularDampingOverride=1
    angularSpringConstant=50
    bAngularSpring=true
    bOverrideAngularDamping=true
    bUseGravityOverride=true
    gravityScale=0.35
    simulationSpace=RootRelative
    alphaBoolBlend {blendOption=Linear}
    boundBone {boneName=hair_r_front_08}
    componentAppliedLinearAccClamp=<800 800 800>
    componentLinearAccScale=<0.5 0.5 0.5>
    componentLinearVelScale=<0.5 0.5 0.5>
    gravityOverride {z=-3000}
    physicsBodyDefinitions [{
      boundBone {boneName=hair_r_front_08}
      boxExtents=<20 20 20>
      constraintSetup {angularLimitsMax=<15 15 15>, angularLimitsMin=<-15 -15 -15>, angularTarget=<1 0 0>}
      localJointOffset=<4 0 0>
     }]
    planarLimits [{drivingBone {boneName=clavicle_r}, planeTransform {location=<0 -7 0>, rotation=<rot 0 0 -120>}}]
   }
   n16 = Anim Dynamics - Bone: hair_r_front_09(ComponentPose=@n15, Alpha=1, ExternalForce="0,0,0") {
    angularDampingOverride=1
    angularSpringConstant=50
    bAngularSpring=true
    bOverrideAngularDamping=true
    bUseGravityOverride=true
    gravityScale=0.35
    simulationSpace=RootRelative
    alphaBoolBlend {blendOption=Linear}
    boundBone {boneName=hair_r_front_09}
    componentAppliedLinearAccClamp=<800 800 800>
    componentLinearAccScale=<0.5 0.5 0.5>
    componentLinearVelScale=<0.5 0.5 0.5>
    gravityOverride {z=-3000}
    physicsBodyDefinitions [{
      boundBone {boneName=hair_r_front_09}
      boxExtents=<20 30 30>
      constraintSetup {angularLimitsMax=<15 15 15>, angularLimitsMin=<-15 -15 -15>, angularTarget=<1 0 0>}
      localJointOffset=<4 0 0>
     }]
    planarLimits [{drivingBone {boneName=clavicle_r}, planeTransform {location=<0 -7 0>, rotation=<rot 0 0 -119.9999>}}]
   }
   n17 = RigidBody!(ComponentPose=Local To Component(LocalPose=Use cached pose 'PostRearHairPhysics'()), Alpha=0.75, ExternalForce) -> (Pose) {
    bOverrideWorldGravity=true
    overridePhysicsAsset=&/F_MED_RollerBladeJacket_FaceAcc_Front_FrontEnd_Physics
    alphaBoolBlend {blendOption=Linear}
    baseBoneRef {boneName=head}
    componentAppliedLinearAccClamp=<400 400 400>
    overrideWorldGravity {z=-980}
    simSpaceSettings {worldAlpha=1}
   }
   n18 = "Transform (Modify) Bone - Bone: hair_r_front_04"(ComponentPose=@n17, Alpha=1, Translation="0,0,0", Rotation="0,0,0", Scale="1,0.8,0.8") {scaleMode=BMM_Additive, alphaBoolBlend {blendOption=Linear}, boneToModify {boneName=hair_r_front_04}, scale {y=0.8, z=0.8}}
   Save cached pose 'JacketFronHairAnimDynamics'(Pose=Blend Poses by bool(BlendPose_0=Component To Local(ComponentPose=@n18), BlendPose_1=Component To Local(ComponentPose=@n16)) {blendType=Linear}) {cacheName=JacketFronHairAnimDynamics}
   n19 = RigidBody!(
    ComponentPose=Local To Component(LocalPose=Use cached pose 'PostRearHairPhysics'())
    Alpha
    OverrideWorldGravity:Vector="0,0,-980"
    ExternalForce
    ComponentLinearAccScale:Vector="0.1,0.1,0.1"
    ComponentLinearVelScale:Vector="0.1,0.1,0.1"
   ) -> (Pose) {
    bClampLinearTranslationLimitToRefPose=true
    bOverrideWorldGravity=true
    overridePhysicsAsset=&/F_MED_RollerBlade_FaceAcc_Front_Skydive_Physics
    simulationSpace=BaseBoneSpace
    alphaBoolBlend {blendOption=Linear}
    baseBoneRef {boneName=head}
    componentAppliedLinearAccClamp=<400 400 400>
    componentLinearAccScale=<0.3 0.3 0.5>
    componentLinearVelScale=<0.3 0.3 0.5>
    overrideWorldGravity {z=-800}
    simSpaceSettings {
     maxAngularAcceleration=50
     maxAngularVelocity=10
     maxLinearAcceleration=50
     maxLinearVelocity=50
     velocityScaleZ=0.2
     worldAlpha=0.25
     externalAngularVelocity=<0.2 0.2 0.2>
     externalLinearDragV=<0.2 0.2 0.2>
     externalLinearVelocity=<0.2 0.2 0.2>
    }
   }
   n20 = Component To Local(ComponentPose=@n19)
   n21 = RigidBody(
    ComponentPose=Local To Component(LocalPose=Use cached pose 'PostRearHairPhysics'())
    OverrideWorldGravity="0,0,-980"
    ComponentLinearAccScale="0.5,0.5,0.5"
    ComponentLinearVelScale="0.5,0.5,0.5"
   ) {
    bClampLinearTranslationLimitToRefPose=true
    bOverrideWorldGravity=true
    overridePhysicsAsset=&/F_MED_RollerBlade_FaceAcc_Front_Physics
    simulationSpace=BaseBoneSpace
    alphaBoolBlend {blendOption=Linear}
    baseBoneRef {boneName=head}
    componentAppliedLinearAccClamp=<400 400 400>
    componentLinearAccScale=<0.3 0.3 0.5>
    componentLinearVelScale=<0.3 0.3 0.5>
    overrideWorldGravity {z=-800}
    simSpaceSettings {
     maxAngularAcceleration=100
     maxAngularVelocity=10
     maxLinearAcceleration=100
     maxLinearVelocity=100
     velocityScaleZ=0.2
     worldAlpha=0.25
     externalAngularVelocity=<0.2 0.2 0.2>
     externalLinearDragV=<0.2 0.2 0.2>
     externalLinearVelocity=<0.2 0.2 0.2>
    }
   }
   n22 = Component To Local(ComponentPose=@n21)
   n23 = RigidBody!(ComponentPose=Local To Component(LocalPose=Use cached pose 'PostRearHairPhysics'()), Alpha, ExternalForce) -> (Pose) {
    bOverrideWorldGravity=true
    overridePhysicsAsset=&/F_MED_RollerBlade_FaceAcc_Front_FrontEnd_Physics
    alphaBoolBlend {blendOption=Linear}
    baseBoneRef {boneName=head}
    componentAppliedLinearAccClamp=<800 800 800>
    overrideWorldGravity {z=-980}
    simSpaceSettings {worldAlpha=1}
   }
   Save cached pose 'FrontHairRigidbodyPhysics'(Pose=Blend Poses by bool(BlendPose_0=Component To Local(ComponentPose=@n23), BlendPose_1=Blend Poses by bool(BlendPose_0=@n20, BlendPose_1=@n22) {blendType=Linear}) {blendType=Linear}) {cacheName=FrontHairRigidbodyPhysics}
  ]}
 CheckBackbling {nodes [
   CheckBackbling() {is=title K2Node_FunctionEntry}
   n1 = Make Array("[0]"=(TagName="Cosmetics.MetaTags.RollerBlade"))
   n2 = Make Array("[0]"=(TagName="Cosmetics.MetaTags.RollerBlade"))
   > n3 = Branch(Condition=AND Boolean(
     A=PickaxeItemDefinitionHasAnyMetaTag(Tags=@n2, bRequiresEquipped=true)
     B=BackpackItemDefinitionHasAnyMetaTag!(Tags:Array of Gameplay Tag Structures=@n1, bRequiresEquipped:bool=false) -> (ReturnValue:bool)
    ))
   > n4 = Branch(Condition=OR Boolean!(A=Get bIsSkyDiving(), B=Get bIsParachuteOpened(), C:bool=Get bIsSkydiveFloating()) -> (ReturnValue))
   > Set Is Wearing Backbling(Is Wearing Backbling=true)
   Set Is Wearing Backbling(execute=@n4.else, Is Wearing Backbling=false)
   Set Is Wearing Backbling(execute=@n3.else, Is Wearing Backbling=CheckForBackbling())
  ]}
 EventGraph {nodes [
   Event BlueprintUpdateAnimation()
   > n1 = Gravity Override State Function DEPRECATED!(
    execute
    CharacterPartAnimInstance=Self-Reference()
    GameTimeInSeconds=GetGameTimeInSeconds()
    GravityOverrideParams_Skydive_GoingUp_28_#1:Gravity Override Setting Struct Structure=(WindFrequency_28_#2=(X=1.5,Y=7,Z=8),WindAmplitude_31_#3=(X=200,Y=-500,Z=400),WindOffset_32_#4=(X=0,Y=0,Z=0),GravityOverride_15_#5=(X=0,Y=-300,Z=1500),JointName_27_#6="root")
    GravityOverrideParams_Skydive_FloatingDown_29_#7:Gravity Override Setting Struct Structure=Get Gravity Override Params Skydive Floating Down() -> (Gravity Override Params Skydive Floating Down:Gravity Override Setting Struct Structure)
    GravityOverrideParams_Skydive_DivingDown_30_#8:Gravity Override Setting Struct Structure=Get Gravity Override Params Skydive Diving Down() -> (Gravity Override Params Skydive Diving Down:Gravity Override Setting Struct Structure)
    GravityOverrideParams_Skydive_Parachuting_31_#9:Gravity Override Setting Struct Structure=Get Gravity Override Params Skydive Parachuting() -> (Gravity Override Params Skydive Parachuting:Gravity Override Setting Struct Structure)
    GravityOverrideParams_OnGround_Standing_32_#10:Gravity Override Setting Struct Structure=(WindFrequency_28_#2=(X=0,Y=0,Z=0),WindAmplitude_31_#3=(X=0,Y=0,Z=0),WindOffset_32_#4=(X=0,Y=0,Z=0),GravityOverride_15_#5=(X=0,Y=0,Z=-980),JointName_27_#6="root")
    GravityOverrideParams_OnGround_Moving_33_#11:Gravity Override Setting Struct Structure=(WindFrequency_28_#2=(X=0,Y=0,Z=0),WindAmplitude_31_#3=(X=0,Y=0,Z=0),WindOffset_32_#4=(X=0,Y=0,Z=0),GravityOverride_15_#5=(X=0,Y=0,Z=-980),JointName_27_#6="root")
    GravityOverrideParams_EmoteOrMelee_34_#12:Gravity Override Setting Struct Structure=(WindFrequency_28_#2=(X=0,Y=0,Z=0),WindAmplitude_31_#3=(X=0,Y=0,Z=0),WindOffset_32_#4=(X=0,Y=0,Z=0),GravityOverride_15_#5=(X=0,Y=0,Z=-980),JointName_27_#6="root")
    GravityOverrideParams_Swimming_35_#13:Gravity Override Setting Struct Structure=(WindFrequency_28_#2=(X=0,Y=0,Z=0),WindAmplitude_31_#3=(X=0,Y=0,Z=0),WindOffset_32_#4=(X=0,Y=0,Z=0),GravityOverride_15_#5=(X=-800,Y=0,Z=0),JointName_27_#6="pelvis")
    GravityOverrideParams_Front_End_36_#14:Gravity Override Setting Struct Structure=(WindFrequency_28_#2=(X=0,Y=0,Z=0),WindAmplitude_31_#3=(X=0,Y=0,Z=0),WindOffset_32_#4=(X=0,Y=0,Z=0),GravityOverride_15_#5=(X=0,Y=0,Z=-980),JointName_27_#6="root")
    GravityOverrideParams_DBNO_40_#15:Gravity Override Setting Struct Structure=(WindFrequency_28_#2=(X=0,Y=0,Z=0),WindAmplitude_31_#3=(X=0,Y=0,Z=0),WindOffset_32_#4=(X=0,Y=0,Z=0),GravityOverride_15_#5=(X=-980,Y=400,Z=0),JointName_27_#6="pelvis")
    GravityOverrideParams_DBNO_Carried_41_#16:Gravity Override Setting Struct Structure=(WindFrequency_28_#2=(X=0,Y=0,Z=0),WindAmplitude_31_#3=(X=0,Y=0,Z=0),WindOffset_32_#4=(X=0,Y=0,Z=0),GravityOverride_15_#5=(X=-980,Y=500,Z=-500),JointName_27_#6="pelvis")
    GravityOverrideParams_OnGround_Falling_45_#17:Gravity Override Setting Struct Structure=(WindFrequency_28_#2=(X=0,Y=0,Z=0),WindAmplitude_31_#3=(X=0,Y=0,Z=0),WindOffset_32_#4=(X=0,Y=0,Z=0),GravityOverride_15_#5=(X=0,Y=0,Z=-980),JointName_27_#6="root")
    GravityOverrideParams_Ziplining_46_#18:Gravity Override Setting Struct Structure=Get Gravity Override Params Ziplining() -> (Gravity Override Params Ziplining:Gravity Override Setting Struct Structure)
    GravityOverrideParams_Skydive_Surfing_49_#19:Gravity Override Setting Struct Structure=(WindFrequency_28_#2=(X=1.5,Y=7,Z=8),WindAmplitude_31_#3=(X=200,Y=-500,Z=400),WindOffset_32_#4=(X=0,Y=0,Z=0),GravityOverride_15_#5=(X=500,Y=-300,Z=-980),JointName_27_#6="root")
    GravityOverrideParams_Riding_Standing_54_#20:Gravity Override Setting Struct Structure=(WindFrequency_28_#2=(X=0,Y=0,Z=0),WindAmplitude_31_#3=(X=0,Y=0,Z=0),WindOffset_32_#4=(X=0,Y=0,Z=0),GravityOverride_15_#5=(X=0,Y=-350,Z=-980),JointName_27_#6="root")
    GravityOverrideParams_Riding_Moving_55_#21:Gravity Override Setting Struct Structure=(WindFrequency_28_#2=(X=0,Y=0,Z=0),WindAmplitude_31_#3=(X=0,Y=0,Z=0),WindOffset_32_#4=(X=0,Y=0,Z=0),GravityOverride_15_#5=(X=0,Y=-450,Z=-450),JointName_27_#6="root")
    GravityOverrideParams_Riding_Moving_Targeting_58_#22:Gravity Override Setting Struct Structure=(WindFrequency_28_#2=(X=0,Y=0,Z=0),WindAmplitude_31_#3=(X=0,Y=0,Z=0),WindOffset_32_#4=(X=0,Y=0,Z=0),GravityOverride_15_#5=(X=0,Y=-200,Z=-980),JointName_27_#6="root")
    GravityOverrideParams_Tactical_Sprint_57_#23:Gravity Override Setting Struct Structure=(WindFrequency_28_#2=(X=0,Y=0,Z=0),WindAmplitude_31_#3=(X=0,Y=0,Z=0),WindOffset_32_#4=(X=0,Y=0,Z=0),GravityOverride_15_#5=(X=0,Y=-450,Z=-500),JointName_27_#6="root")
    GravityOverrideParams_Rail_Grinding_61_#24:Gravity Override Setting Struct Structure=(WindFrequency_28_#2=(X=0,Y=0,Z=0),WindAmplitude_31_#3=(X=0,Y=0,Z=0),WindOffset_32_#4=(X=0,Y=0,Z=0),GravityOverride_15_#5=(X=0,Y=0,Z=-980),JointName_27_#6="root")
    GravityOverrideParams_DBNO_Tumbling_67_#25:Gravity Override Setting Struct Structure=(WindFrequency_28_#2=(X=0,Y=0,Z=0),WindAmplitude_31_#3=(X=0,Y=0,Z=0),WindOffset_32_#4=(X=0,Y=0,Z=0),GravityOverride_15_#5=(X=-2000,Y=0,Z=0),JointName_27_#6="pelvis")
    GravityOverrideParams_DBNO_RollFwd_68_#26:Gravity Override Setting Struct Structure=(WindFrequency_28_#2=(X=0,Y=0,Z=0),WindAmplitude_31_#3=(X=0,Y=0,Z=0),WindOffset_32_#4=(X=0,Y=0,Z=0),GravityOverride_15_#5=(X=-2000,Y=0,Z=0),JointName_27_#6="pelvis")
    GravityOverrideParams_DBNO_RollBwd_69_#27:Gravity Override Setting Struct Structure=(WindFrequency_28_#2=(X=0,Y=0,Z=0),WindAmplitude_31_#3=(X=0,Y=0,Z=0),WindOffset_32_#4=(X=0,Y=0,Z=0),GravityOverride_15_#5=(X=-4000,Y=0,Z=0),JointName_27_#6="pelvis")
    GravityOverrideParams_DBNO_Impulse_70_#28:Gravity Override Setting Struct Structure=(WindFrequency_28_#2=(X=0,Y=0,Z=0),WindAmplitude_31_#3=(X=0,Y=0,Z=0),WindOffset_32_#4=(X=0,Y=0,Z=0),GravityOverride_15_#5=(X=-2000,Y=0,Z=0),JointName_27_#6="pelvis")
    RBAN Params
    DebugDraw
    DebugJointName
   ) -> (then, GravityOverride, CurrentStateOut, GravityOverrideSettingsOut, RBAN SettingsOut)
   > Set Gravity Override(Gravity Override=@n1)
   n2 = OR Boolean(A=Get bIsSkyDiving(), B=Get bIsParachuteOpened(), C=Get bIsSkydiveFloating())
   > Set SkydivingOrParachuting(SkydivingOrParachuting:bool=@n2) -> (Output_Get:bool)
   > CheckBackbling() {is=title K2Node_CallFunction}
   Event OnCharacterCustomizationLoaded()
   > n3 = Sequence!(execute) -> (then_0, then_1, then_2:Exec)
   > n4 = Branch(Condition=CharacterItemDefinitionHasAnyMetaTag(Tags=Make Array("[0]"=(TagName="Cosmetics.Variant.Property.Stage2"))))
   > Set WearingJacket(WearingJacket:bool=true) -> (Output_Get:bool)
   Set WearingJacket(execute=@n4.else, WearingJacket=false)
   CheckBackbling(execute=@n3.then_1)
   n5 = Cast To PlayerPawn_Athena(execute=@n3.then_2, Object=TryGetPawnOwner())
   > Set BodyPart(BodyPart:Skeletal Mesh Component Object Reference=GetSkeletalMeshForPartType(self=@n5)) -> (Output_Get:Skeletal Mesh Component Object Reference)
   Set NewVar(NewVar:Fort Pawn Object Reference=TryGetFortPawn()) -> (Output_Get:Fort Pawn Object Reference)
   n6 = Get NewVar() -> (NewVar:Fort Pawn Object Reference)
   > Is Valid(InputObject=@n6)
   n7 = Get CurrentWeapon!(self:Fort Pawn Object Reference=@n6) -> (CurrentWeapon:Fort Weapon Object Reference)
   > Set NewVar_0(NewVar_0:bool=AND Boolean(A=IsValid(Object=@n7), B=IsWeaponHolstered!(self:Fort Pawn Object Reference=@n6) -> (ReturnValue:bool))) -> (Output_Get:bool)
   n8 = Get NewVar()
   > Branch(Condition=WeaponDataHasMatchingGameplayTag!(self:Fort Weapon Object Reference=Get CurrentWeapon(self=@n8), TagToCheck:Gameplay Tag Structure) -> (ReturnValue:bool))
   CheckBackbling()
   Event OnWeaponChanged()
   Branch(Condition=@n2)
  ]}
]
variables [
 BodyPart {type=SkeletalMeshComponent}
 Gravity Override {type=Vector}
 Gravity Override Params Skydive Diving Down {
  type=GravityOverrideSettingStruct
  default {jointName=root, gravityOverride=<0 -700 600>, windAmplitude=<100 -300 300>, windFrequency=<1.5 7 8>, windOffset=<0 -300 250>}
 }
 Gravity Override Params Skydive Floating Down {
  type=GravityOverrideSettingStruct
  default {jointName=root, gravityOverride=<0 -450 -100>, windAmplitude=<100 -250 200>, windFrequency=<1.5 7 8>, windOffset=<0 -300 200>}
 }
 Gravity Override Params Skydive Parachuting {
  type=GravityOverrideSettingStruct
  default {jointName=root, gravityOverride=<0 -800 300>, windAmplitude=<200 -500 400>, windFrequency=<1.5 7 8>, windOffset=<0 0 0>}
 }
 Gravity Override Params Ziplining {
  type=GravityOverrideSettingStruct
  default {jointName=root, gravityOverride=<0 0 -980>, windAmplitude=<0 0 0>, windFrequency=<0 0 0>, windOffset=<0 0 0>}
 }
 Is Backbling A Cape {type=boolean}
 Is Wearing Backbling {type=boolean}
 NewVar {type=FortPawn, default=null}
 NewVar_0 {type=boolean}
 SkydivingOrParachuting {type=boolean}
 WearingJacket {type=boolean}
]
